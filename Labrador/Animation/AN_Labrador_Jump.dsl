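; Anim notify for the Labrador jump that plays a training-dependent sound when it fires during an animation.
; It casts the mesh's owner to BP_AIPawn_Labrador_Ridable and picks GoldenSound once TrainingPercentage reaches 1.0, otherwise NormalSound.
; The chosen sound plays at the pawn's location with that pawn as owning actor, and a debug "Hello" string is printed.
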(fn Received_Notify
  (params (MeshComp :type Skeletal%20Mesh%20Component%20Object%20Reference) (Animation :type Anim%20Sequence%20Base%20Object%20Reference) (EventReference :type Anim%20Notify%20Event%20Reference%20Structure%20%28by%20ref%29))
  (bind _asbp_aipawn_labrador_ridable (Utilities|Casting|CastToBP_AIPawn_Labrador_Ridable (Components|GetOwner MeshComp)))
  (Audio|PlaySoundatLocation (select (>= (Variables|/Labrador/Pawn/BP_AIPawn_Labrador_Ridable.BP_AIPawn_Labrador_Ridable_C|GetTrainingPercentage _asbp_aipawn_labrador_ridable) 1.0) (Variables|Default|GetGoldenSound) (Variables|Default|GetNormalSound)) (Transformation|GetActorLocation _asbp_aipawn_labrador_ridable) "0, 0, 0" 1.0 1.0 0.0 0 0 _asbp_aipawn_labrador_ridable)
  (Development|PrintString "Hello"))
