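# Animation Blueprint for the Labrador, a rideable dog creature.
# Bools switch between a floating panic pose, a ridden state machine with FPWeight-scaled lean additive, and an unridden one; both machines cover idle variants, movement, turning, jumping, swimming, startle and topple.
# A LootDrop machine adds a hit-react while leaking, before DefaultSlot and final inertialization.
# Movement uses weighted head look-at, and topple uses quadruped slope warping.
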AnimBlueprint /Labrador/Animation/AnimBlueprint/Labrador_AnimBP
target_skeleton=/Labrador/Meshes/Labrador_Empty/Labrador_Empty_Skeleton
graphs [
 AnimGraph {nodes [
   AnimGraphNode_StateMachine_0 = Riding Locomotion() {is=title AnimGraphNode_StateMachine, maxTransitionsPerFrame=1}
   n1 = Apply Additive(
    Base=@AnimGraphNode_StateMachine_0
    Additive="Sequence Player (additive) 'Labrador_Lean_Corrective'"() {bLoopAnimation=false, sequence=/Labrador/Animation/Labrador_Lean_Corrective}
    Alpha=Get FPWeight() -> (FPWeight:float)
   )
   AnimGraphNode_StateMachine_2 = Locomotion() {is=title AnimGraphNode_StateMachine, maxTransitionsPerFrame=1}
   n2 = Get bIsBeingRidden() -> (bIsBeingRidden:bool)
   AnimGraphNode_StateMachine_3 = LootDrop() {is=title AnimGraphNode_StateMachine}
   n3 = Apply Additive(
    Base=Blend Poses by bool(
     BlendPose_0=Sequence Player 'Labrador_Panic'() {sequence=/Labrador/Animation/Labrador_Panic}
     BlendPose_1=Blend Poses by bool(BlendPose_0=@n1, BlendPose_1=@AnimGraphNode_StateMachine_2, bActiveValue=@n2)
     BlendTime_0=0.5
     BlendTime_1=0.5
     bActiveValue=Get bIsFloating() -> (bIsFloating:bool)
    ) {blendType=Linear, transitionType=Inertialization}
    Additive=@AnimGraphNode_StateMachine_3
   ) {alphaBoolBlend {blendOption=Linear}}
   Output Pose(Result=Inertialization!(Source:Pose Link Structure=Slot 'DefaultSlot'(Source=@n3)) -> (Pose:Pose Link Structure) {is=title AnimGraphNode_Inertialization}) {layerGroup=None}
  ]}
 Riding Locomotion {
  entry=Idle
  path=AnimGraph.AnimGraphNode_StateMachine_0.Riding Locomotion
  ids=id
  nodes [
   Swim State Alias() {is=title AnimStateAliasNode, also {aliases [Idle,JumpLand,JumpLoop,Movement]}}
   Topple Alias() {is=title AnimStateAliasNode, also {aliases [Jumping,JumpLand,JumpLoop,Movement,Startled]}}
   Jump Alias() {is=title AnimStateAliasNode, also {aliases [Idle, JumpLand, Movement, Startled, Swimming, Turn in Place]}}
   AnimStateNode_0 = Turn in Place() {is=title AnimStateNode}
   AnimStateNode_1 = Topple_Idle() {is=title AnimStateNode}
   AnimStateNode_2 = Jumping() {is=title AnimStateNode, bAlwaysResetOnEntry=true}
   AnimStateNode_4 = Idle()
   AnimStateNode_5 = JumpLoop() {is=title AnimStateNode}
   AnimStateNode_6 = JumpLand() {is=title AnimStateNode, stateEntered {guid=#1, notifyName=LandEntered}}
   AnimStateNode_7 = Startled() {is=title AnimStateNode}
   AnimStateNode_8 = Movement() {is=title AnimStateNode}
   AnimStateNode_9 = Swimming() {is=title AnimStateNode}
   AnimStateTransitionNode_0 = AnimStateTransitionNode() {bAllowInertializationForSelfTransitions=true}
   AnimStateTransitionNode_1 = AnimStateTransitionNode() {logicType=TLT_Inertialization, priorityOrder=2}
   AnimStateTransitionNode_10 = AnimStateTransitionNode() {bAutomaticRuleBasedOnSequencePlayerInState=true, crossfadeDuration=0.3, logicType=TLT_Inertialization, priorityOrder=3}
   AnimStateTransitionNode_11 = AnimStateTransitionNode() {crossfadeDuration=0.4, logicType=TLT_Inertialization, priorityOrder=4}
   AnimStateTransitionNode_12 = AnimStateTransitionNode() {bAutomaticRuleBasedOnSequencePlayerInState=true, crossfadeDuration=0.15, logicType=TLT_Inertialization}
   AnimStateTransitionNode_13 = AnimStateTransitionNode() {bAllowInertializationForSelfTransitions=true}
   AnimStateTransitionNode_16 = AnimStateTransitionNode() {crossfadeDuration=0.4, logicType=TLT_Inertialization}
   AnimStateTransitionNode_19 = AnimStateTransitionNode() {bAllowInertializationForSelfTransitions=true}
   AnimStateTransitionNode_2 = AnimStateTransitionNode() {crossfadeDuration=0.35, logicType=TLT_Inertialization}
   AnimStateTransitionNode_20 = AnimStateTransitionNode() {logicType=TLT_Inertialization, priorityOrder=100}
   AnimStateTransitionNode_29 = AnimStateTransitionNode() {bAllowInertializationForSelfTransitions=true}
   AnimStateTransitionNode_3 = AnimStateTransitionNode() {crossfadeDuration=0.1, logicType=TLT_Inertialization, priorityOrder=3}
   AnimStateTransitionNode_30 = AnimStateTransitionNode() {bAllowInertializationForSelfTransitions=true}
   AnimStateTransitionNode_4 = AnimStateTransitionNode() {bAllowInertializationForSelfTransitions=true, priorityOrder=2}
   AnimStateTransitionNode_5 = AnimStateTransitionNode() {bAllowInertializationForSelfTransitions=true}
   AnimStateTransitionNode_6 = AnimStateTransitionNode() {bAllowInertializationForSelfTransitions=true}
   AnimStateTransitionNode_7 = AnimStateTransitionNode() {bAllowInertializationForSelfTransitions=true}
   AnimStateTransitionNode_9 = AnimStateTransitionNode() {crossfadeDuration=0.4, logicType=TLT_Inertialization}
  ]
  links [
   {
    dst=Movement
    rule=AnimGraph.AnimGraphNode_StateMachine_0.Riding Locomotion.AnimStateTransitionNode_13.Transition
    src=Idle
    via=AnimStateTransitionNode_13
   }
   {
    dst=Startled
    rule=AnimGraph.AnimGraphNode_StateMachine_0.Riding Locomotion.AnimStateTransitionNode_3.Transition
    src=Idle
    via=AnimStateTransitionNode_3
   }
   {
    dst=Turn in Place
    rule=AnimGraph.AnimGraphNode_StateMachine_0.Riding Locomotion.AnimStateTransitionNode_29.Transition
    src=Idle
    via=AnimStateTransitionNode_29
   }
   {
    dst=JumpLoop
    rule=AnimGraph.AnimGraphNode_StateMachine_0.Riding Locomotion.AnimStateTransitionNode_4.Transition
    src=Jump Alias
    via=AnimStateTransitionNode_4
   }
   {
    dst=Jumping
    rule=AnimGraph.AnimGraphNode_StateMachine_0.Riding Locomotion.AnimStateTransitionNode_5.Transition
    src=Jump Alias
    via=AnimStateTransitionNode_5
   }
   {
    dst=Idle
    rule=AnimGraph.AnimGraphNode_StateMachine_0.Riding Locomotion.AnimStateTransitionNode_10.Transition
    src=JumpLand
    via=AnimStateTransitionNode_10
   }
   {
    dst=Movement
    rule=AnimGraph.AnimGraphNode_StateMachine_0.Riding Locomotion.AnimStateTransitionNode_11.Transition
    src=JumpLand
    via=AnimStateTransitionNode_11
   }
   {
    dst=JumpLand
    rule=AnimGraph.AnimGraphNode_StateMachine_0.Riding Locomotion.AnimStateTransitionNode_9.Transition
    src=JumpLoop
    via=AnimStateTransitionNode_9
   }
   {
    dst=JumpLoop
    rule=AnimGraph.AnimGraphNode_StateMachine_0.Riding Locomotion.AnimStateTransitionNode_0.Transition
    src=Jumping
    via=AnimStateTransitionNode_0
   }
   {
    dst=Idle
    rule=AnimGraph.AnimGraphNode_StateMachine_0.Riding Locomotion.AnimStateTransitionNode_1.Transition
    src=Movement
    via=AnimStateTransitionNode_1
   }
   {
    dst=Movement
    rule=AnimGraph.AnimGraphNode_StateMachine_0.Riding Locomotion.AnimStateTransitionNode_12.Transition
    src=Startled
    via=AnimStateTransitionNode_12
   }
   {
    dst=Swimming
    rule=AnimGraph.AnimGraphNode_StateMachine_0.Riding Locomotion.AnimStateTransitionNode_6.Transition
    src=Swim State Alias
    via=AnimStateTransitionNode_6
   }
   {
    dst=Movement
    rule=AnimGraph.AnimGraphNode_StateMachine_0.Riding Locomotion.AnimStateTransitionNode_16.Transition
    src=Swimming
    via=AnimStateTransitionNode_16
   }
   {
    dst=Topple_Idle
    rule=AnimGraph.AnimGraphNode_StateMachine_0.Riding Locomotion.AnimStateTransitionNode_19.Transition
    src=Topple Alias
    via=AnimStateTransitionNode_19
   }
   {
    dst=Topple_Idle
    rule=AnimGraph.AnimGraphNode_StateMachine_0.Riding Locomotion.AnimStateTransitionNode_7.Transition
    src=Topple Alias
    via=AnimStateTransitionNode_7
   }
   {
    dst=JumpLoop
    rule=AnimGraph.AnimGraphNode_StateMachine_0.Riding Locomotion.AnimStateTransitionNode_20.Transition
    src=Topple_Idle
    via=AnimStateTransitionNode_20
   }
   {
    dst=Startled
    rule=AnimGraph.AnimGraphNode_StateMachine_0.Riding Locomotion.AnimStateTransitionNode_2.Transition
    src=Topple_Idle
    via=AnimStateTransitionNode_2
   }
   {
    dst=Idle
    rule=AnimGraph.AnimGraphNode_StateMachine_0.Riding Locomotion.AnimStateTransitionNode_30.Transition
    src=Turn in Place
    via=AnimStateTransitionNode_30
   }
  ]
 }
 Turn in Place {
  path=AnimGraph.AnimGraphNode_StateMachine_0.Riding Locomotion.AnimStateNode_0.Turn in Place
  nodes [Output Animation Pose(Result=Sequence Player 'Labrador_Walk_N'() {groupName=Locomotion, method=SyncGroup, playRate=1.2, sequence=/Labrador/Animation/Labrador_Walk_N})]
 }
 Topple_Idle {
  path=AnimGraph.AnimGraphNode_StateMachine_0.Riding Locomotion.AnimStateNode_1.Topple_Idle
  nodes [
   AnimGraphNode_StateMachine_0 = LabradorTopple() {is=title AnimGraphNode_StateMachine}
   n1 = Quadruped Slope Warping!(ComponentPose:Component Space Pose Link Structure=Local To Component(LocalPose=@AnimGraphNode_StateMachine_0), Alpha:float=1) -> (Pose:Component Space Pose Link Structure) {
    is=title FortAnimGraphNode_QuadrupedSlopeWarping
    bOnlyAffectSkeletonRootBone=true
    bUseMovementComponentFloorInfoInsteadOfTraces=true
    maxStepHeight=35
    alphaBoolBlend {blendOption=Linear}
    armsFeetDefinitions [
     {numBonesInLimb=2, fKFootBone {boneName=frontleg_end_l}, iKFootBone {boneName=frontleg_04_l}}
     {numBonesInLimb=2, fKFootBone {boneName=frontleg_end_r}, iKFootBone {boneName=frontleg_end_r}}
    ]
    armsIKFootRootBone {boneName=lower_body}
    armsRootBone {boneName=upper_body}
    feetDefinitions [
     {numBonesInLimb=2, fKFootBone {boneName=backleg_end_l}, iKFootBone {boneName=backleg_end_l}}
     {numBonesInLimb=2, fKFootBone {boneName=backleg_end_r}, iKFootBone {boneName=backleg_end_r}}
    ]
    iKFootRootBone {boneName=lower_body}
    pelvisBone {boneName=lower_body}
    skeletonRootBone {boneName=root}
   }
   Output Animation Pose(Result=Component To Local(ComponentPose=@n1))
  ]
 }
 LabradorTopple {
  entry=GroundToSitupIdle
  path=AnimGraph.AnimGraphNode_StateMachine_0.Riding Locomotion.AnimStateNode_1.Topple_Idle.AnimGraphNode_StateMachine_0.LabradorTopple
  ids=id
  nodes [
   AnimStateNode_1 = GroundToSitupIdle() {is=title AnimStateNode}
   AnimStateNode_2 = SitUpIdle() {is=title AnimStateNode}
   AnimStateTransitionNode_1 = AnimStateTransitionNode() {crossfadeDuration=0.5, logicType=TLT_Inertialization}
  ]
  links [{
    dst=SitUpIdle
    rule=AnimGraph.AnimGraphNode_StateMachine_0.Riding Locomotion.AnimStateNode_1.Topple_Idle.AnimGraphNode_StateMachine_0.LabradorTopple.AnimStateTransitionNode_1.Transition
    src=GroundToSitupIdle
    via=AnimStateTransitionNode_1
   }]
 }
 GroundToSitupIdle {
  path=AnimGraph.AnimGraphNode_StateMachine_0.Riding Locomotion.AnimStateNode_1.Topple_Idle.AnimGraphNode_StateMachine_0.LabradorTopple.AnimStateNode_1.GroundToSitupIdle
  nodes [Output Animation Pose(Result=Sequence Player 'ToppleToDizzy_Anim'() {sequence=/Labrador/Animation/ToppleToDizzy_Anim})]
 }
 SitUpIdle {
  path=AnimGraph.AnimGraphNode_StateMachine_0.Riding Locomotion.AnimStateNode_1.Topple_Idle.AnimGraphNode_StateMachine_0.LabradorTopple.AnimStateNode_2.SitUpIdle
  nodes [Output Animation Pose(Result=Sequence Player 'Labrador_Dizzy'() {sequence=/Labrador/Animation/Labrador_Dizzy})]
 }
 Transition {
  path=AnimGraph.AnimGraphNode_StateMachine_0.Riding Locomotion.AnimStateNode_1.Topple_Idle.AnimGraphNode_StateMachine_0.LabradorTopple.AnimStateTransitionNode_1.Transition
  nodes [Result(bCanEnterTransition=Get bIsInterruptible() -> (bIsInterruptible:bool))]
 }
 Jumping {
  path=AnimGraph.AnimGraphNode_StateMachine_0.Riding Locomotion.AnimStateNode_2.Jumping
  nodes [Output Animation Pose(Result=Sequence Player 'Labrador_JumpUp_Mounted'() {bLoopAnimation=false, sequence=/Labrador/Animation/Labrador_JumpUp_Mounted})]
 }
 Idle {
  path=AnimGraph.AnimGraphNode_StateMachine_0.Riding Locomotion.AnimStateNode_4.Idle
  nodes [
   n1 = Random Sequence Player!() -> (Pose:Pose Link Structure) {
    is=title AnimGraphNode_RandomPlayer
    entries [
     {
      chanceToPlay=0.5
      maxLoopCount=7
      maxPlayRate=1
      minLoopCount=3
      minPlayRate=1
      sequence=/Labrador/Animation/Labrador_Idle
      blendIn {blendOption=Linear, blendTime=0, customCurve=null}
     }
     {
      chanceToPlay=0.2
      maxLoopCount=1
      maxPlayRate=1
      minLoopCount=1
      minPlayRate=1
      sequence=/Labrador/Animation/Labrador_Idle_Ear_Twitch
      blendIn {blendOption=Linear, blendTime=0, customCurve=null}
     }
     {
      chanceToPlay=0.1
      maxLoopCount=1
      maxPlayRate=1
      minLoopCount=1
      minPlayRate=1
      sequence=/Labrador/Animation/Labrador_Idle_Head_Twitch
      blendIn {blendOption=Linear, blendTime=0, customCurve=null}
     }
     {
      chanceToPlay=0.1
      maxLoopCount=1
      maxPlayRate=1
      minLoopCount=1
      minPlayRate=1
      sequence=/Labrador/Animation/Labrador_Idle_Tail_Twitch
      blendIn {blendOption=Linear, blendTime=0, customCurve=null}
     }
     {
      chanceToPlay=0.1
      maxLoopCount=1
      maxPlayRate=1
      minLoopCount=1
      minPlayRate=1
      sequence=/Labrador/Animation/Labrador_Mouth_Twitch
      blendIn {blendOption=Linear, blendTime=0, customCurve=null}
     }
    ]
   }
   Output Animation Pose(Result=@n1)
  ]
 }
 JumpLoop {
  path=AnimGraph.AnimGraphNode_StateMachine_0.Riding Locomotion.AnimStateNode_5.JumpLoop
  nodes [Output Animation Pose(Result=Sequence Player 'Labrador_FallLoop_Mounted'() {sequence=/Labrador/Animation/Labrador_FallLoop_Mounted})]
 }
 JumpLand {
  path=AnimGraph.AnimGraphNode_StateMachine_0.Riding Locomotion.AnimStateNode_6.JumpLand
  nodes [Output Animation Pose(Result=Sequence Player 'Labrador_Land_Mounted'() {bLoopAnimation=false, sequence=/Labrador/Animation/Labrador_Land_Mounted})]
 }
 Startled {
  path=AnimGraph.AnimGraphNode_StateMachine_0.Riding Locomotion.AnimStateNode_7.Startled
  nodes [Output Animation Pose(Result=Sequence Player 'Labrador_Startle_Mounted'() {bLoopAnimation=false, sequence=/Labrador/Animation/Labrador_Startle_Mounted})]
 }
 Movement {
  path=AnimGraph.AnimGraphNode_StateMachine_0.Riding Locomotion.AnimStateNode_8.Movement
  nodes [
   Blend Poses by bool(
    BlendPose_0=Sequence Player 'Labrador_Run_Bounding'!(PlayRate:float=1.4) -> (Pose) {playRate=1.3, sequence=/Labrador/Animation/Labrador_Run_Bounding}
    BlendPose_1=Sequence Player 'Labrador_Run_Bounding'(PlayRate=1.2) {playRate=1.3, sequence=&Labrador_Run_Bounding}
    BlendTime_0=0.6
    BlendTime_1=0.6
    bActiveValue=Get bTargetIsClose() -> (bTargetIsClose:bool)
   ) {blendType=Linear, transitionType=Inertialization}
   n1 = Sequence Player 'Labrador_Walk_N'() {groupName=Locomotion, method=SyncGroup, playRate=1.2, sequence=&Labrador_Walk_N}
   n2 = Sequence Player 'Labrador_Run_Bounding_Mounted'!(PlayRate:float=1.2) -> (Pose) {sequence=/Labrador/Animation/Labrador_Run_Bounding_Mounted}
   n3 = Get bIsRunning() -> (bIsRunning:bool)
   n4 = Weighted Look At - Bone: head!(
    ComponentPose:Component Space Pose Link Structure=Local To Component(LocalPose=Blend Poses by bool(BlendPose_0=@n2, BlendPose_1=@n1, BlendTime_0=0.5, BlendTime_1=0.7, bActiveValue=@n3) {blendType=Linear, transitionType=Inertialization})
    Alpha:float=Get LookAtAlpha() -> (LookAtAlpha:float)
    LookAtLocation:Vector=Get AimTargetLocation() -> (AimTargetLocation:Vector)
   ) -> (Pose:Component Space Pose Link Structure) {
    is=title FortAnimGraphNode_WeightedLookAt
    interpolationTime=1
    interpolationType=EaseInOutExponent3
    lookAtClamp=60
    alphaBoolBlend {blendOption=Linear}
    boneToModify {boneName=head}
    boneWeightsToApply [{weight=0.8}, {weight=0.1}, {weight=0.1}]
    lookAt_Axis {bInLocalSpace=false, axis {y=0}}
    lookUp_Axis {bInLocalSpace=false, axis {x=0}}
   }
   Output Animation Pose(Result=Component To Local(ComponentPose=@n4))
  ]
 }
 Swimming {
  path=AnimGraph.AnimGraphNode_StateMachine_0.Riding Locomotion.AnimStateNode_9.Swimming
  nodes [Output Animation Pose(Result=Sequence Player 'Labrador_Swim_N'() {playRate=1.4, sequence=/Labrador/Animation/Labrador_Swim_N})]
 }
 Transition {
  path=AnimGraph.AnimGraphNode_StateMachine_0.Riding Locomotion.AnimStateTransitionNode_0.Transition
  nodes [Result(bCanEnterTransition=Get bIsFalling())]
 }
 Transition {
  path=AnimGraph.AnimGraphNode_StateMachine_0.Riding Locomotion.AnimStateTransitionNode_1.Transition
  nodes [Result(bCanEnterTransition=Get bLocoToIdle() -> (bLocoToIdle:bool))]
 }
 Transition {path=AnimGraph.AnimGraphNode_StateMachine_0.Riding Locomotion.AnimStateTransitionNode_10.Transition, nodes [Result()]}
 Transition {
  path=AnimGraph.AnimGraphNode_StateMachine_0.Riding Locomotion.AnimStateTransitionNode_11.Transition
  nodes [Result(bCanEnterTransition=Get bLandToMovement() -> (bLandToMovement:bool))]
 }
 Transition {
  path=AnimGraph.AnimGraphNode_StateMachine_0.Riding Locomotion.AnimStateTransitionNode_12.Transition
  nodes [
   Result(bCanEnterTransition=Get bIsInterruptible())
   AND Boolean(
    A="float < float"!(A:double="Get Relevant Anim Time Remaining Fraction (Startled)"!() -> (ReturnValue:float) {is=title K2Node_AnimGetter}, B:double=0.9) -> (ReturnValue:bool) {is=title K2Node_CallFunction}
    B=false
   )
  ]
 }
 Transition {
  path=AnimGraph.AnimGraphNode_StateMachine_0.Riding Locomotion.AnimStateTransitionNode_13.Transition
  nodes [Result(bCanEnterTransition=Get bIdleToMovement() -> (bIdleToMovement:bool))]
 }
 Transition {
  path=AnimGraph.AnimGraphNode_StateMachine_0.Riding Locomotion.AnimStateTransitionNode_16.Transition
  nodes [Result(bCanEnterTransition=NOT Boolean(A=Get bIsSurfaceSwimming()))]
 }
 Transition {
  path=AnimGraph.AnimGraphNode_StateMachine_0.Riding Locomotion.AnimStateTransitionNode_19.Transition
  nodes [Result(bCanEnterTransition=Get bIsToppled() -> (bIsToppled:bool))]
 }
 Transition {
  path=AnimGraph.AnimGraphNode_StateMachine_0.Riding Locomotion.AnimStateTransitionNode_2.Transition
  nodes [Result(bCanEnterTransition=NOT Boolean(A=Get bIsToppled()))]
 }
 Transition {
  path=AnimGraph.AnimGraphNode_StateMachine_0.Riding Locomotion.AnimStateTransitionNode_20.Transition
  nodes [Result(bCanEnterTransition=Get bHighVelocity() -> (bHighVelocity:bool))]
 }
 Transition {
  path=AnimGraph.AnimGraphNode_StateMachine_0.Riding Locomotion.AnimStateTransitionNode_29.Transition
  nodes [Result(bCanEnterTransition=Get bIsTurningInPlace() -> (bIsTurningInPlace:bool))]
 }
 Transition {
  path=AnimGraph.AnimGraphNode_StateMachine_0.Riding Locomotion.AnimStateTransitionNode_3.Transition
  nodes [Result(bCanEnterTransition=Get bIdleToStartled() -> (bIdleToStartled:bool))]
 }
 Transition {
  path=AnimGraph.AnimGraphNode_StateMachine_0.Riding Locomotion.AnimStateTransitionNode_30.Transition
  nodes [Result(bCanEnterTransition=NOT Boolean(A=Get bIsTurningInPlace()))]
 }
 Transition {
  path=AnimGraph.AnimGraphNode_StateMachine_0.Riding Locomotion.AnimStateTransitionNode_4.Transition
  nodes [Result(bCanEnterTransition=Get bIsFalling())]
 }
 Transition {
  path=AnimGraph.AnimGraphNode_StateMachine_0.Riding Locomotion.AnimStateTransitionNode_5.Transition
  nodes [Result(bCanEnterTransition=Get bIsJumping())]
 }
 Transition {
  path=AnimGraph.AnimGraphNode_StateMachine_0.Riding Locomotion.AnimStateTransitionNode_6.Transition
  nodes [Result(bCanEnterTransition=Get bIsSurfaceSwimming())]
 }
 Transition {
  path=AnimGraph.AnimGraphNode_StateMachine_0.Riding Locomotion.AnimStateTransitionNode_7.Transition
  nodes [Result(bCanEnterTransition=Get bToppledOnGround() -> (bToppledOnGround:bool))]
 }
 Transition {
  path=AnimGraph.AnimGraphNode_StateMachine_0.Riding Locomotion.AnimStateTransitionNode_9.Transition
  nodes [Result(bCanEnterTransition=NOT Boolean(A=Get bIsFalling()))]
 }
 Locomotion {
  entry=Idle
  path=AnimGraph.AnimGraphNode_StateMachine_2.Locomotion
  ids=id
  nodes [
   Swim State Alias() {also {aliases [Idle,JumpLand,JumpLoop,Movement]}}
   Topple Alias() {also {aliases [Jumping,JumpLand,JumpLoop,Movement,Startled]}}
   AnimStateNode_0 = Turn in Place()
   AnimStateNode_1 = Topple_Idle()
   AnimStateNode_2 = Jumping()
   AnimStateNode_4 = Idle()
   AnimStateNode_5 = JumpLoop()
   AnimStateNode_6 = JumpLand()
   AnimStateNode_7 = Startled()
   AnimStateNode_8 = Movement()
   AnimStateNode_9 = Swimming()
   AnimStateTransitionNode_0 = AnimStateTransitionNode() {logicType=TLT_Inertialization}
   AnimStateTransitionNode_1 = AnimStateTransitionNode() {logicType=TLT_Inertialization, priorityOrder=2}
   AnimStateTransitionNode_10 = AnimStateTransitionNode() {crossfadeDuration=0.3, logicType=TLT_Inertialization, priorityOrder=3}
   AnimStateTransitionNode_11 = AnimStateTransitionNode() {crossfadeDuration=0.4, logicType=TLT_Inertialization, priorityOrder=4}
   AnimStateTransitionNode_12 = AnimStateTransitionNode() {bAutomaticRuleBasedOnSequencePlayerInState=true, crossfadeDuration=0.15, logicType=TLT_Inertialization}
   AnimStateTransitionNode_13 = AnimStateTransitionNode() {crossfadeDuration=0.4, logicType=TLT_Inertialization, priorityOrder=100}
   AnimStateTransitionNode_16 = AnimStateTransitionNode() {crossfadeDuration=0.4, logicType=TLT_Inertialization}
   AnimStateTransitionNode_17 = AnimStateTransitionNode() {crossfadeDuration=0.4, logicType=TLT_Inertialization, priorityOrder=100}
   AnimStateTransitionNode_19 = AnimStateTransitionNode() {bAllowInertializationForSelfTransitions=true}
   AnimStateTransitionNode_2 = AnimStateTransitionNode() {crossfadeDuration=0.35, logicType=TLT_Inertialization}
   AnimStateTransitionNode_20 = AnimStateTransitionNode() {logicType=TLT_Inertialization, priorityOrder=100}
   AnimStateTransitionNode_26 = AnimStateTransitionNode() {crossfadeDuration=0.4, logicType=TLT_Inertialization, priorityOrder=2}
   AnimStateTransitionNode_27 = AnimStateTransitionNode() {crossfadeDuration=0.4, logicType=TLT_Inertialization}
   AnimStateTransitionNode_29 = AnimStateTransitionNode() {bAllowInertializationForSelfTransitions=true}
   AnimStateTransitionNode_3 = AnimStateTransitionNode() {crossfadeDuration=0.1, logicType=TLT_Inertialization, priorityOrder=3}
   AnimStateTransitionNode_30 = AnimStateTransitionNode() {bAllowInertializationForSelfTransitions=true}
   AnimStateTransitionNode_4 = AnimStateTransitionNode() {crossfadeDuration=0.3, logicType=TLT_Inertialization, priorityOrder=2}
   AnimStateTransitionNode_6 = AnimStateTransitionNode() {bAllowInertializationForSelfTransitions=true}
   AnimStateTransitionNode_7 = AnimStateTransitionNode() {bAllowInertializationForSelfTransitions=true}
   AnimStateTransitionNode_8 = AnimStateTransitionNode() {crossfadeDuration=0.5, logicType=TLT_Inertialization}
   AnimStateTransitionNode_9 = AnimStateTransitionNode() {crossfadeDuration=0.4, logicType=TLT_Inertialization}
  ]
  links [
   {
    dst=JumpLoop
    rule=AnimGraph.AnimGraphNode_StateMachine_2.Locomotion.AnimStateTransitionNode_26.Transition
    src=Idle
    via=AnimStateTransitionNode_26
   }
   {
    dst=Jumping
    rule=AnimGraph.AnimGraphNode_StateMachine_2.Locomotion.AnimStateTransitionNode_0.Transition
    src=Idle
    via=AnimStateTransitionNode_0
   }
   {
    dst=Startled
    rule=AnimGraph.AnimGraphNode_StateMachine_2.Locomotion.AnimStateTransitionNode_3.Transition
    src=Idle
    via=AnimStateTransitionNode_3
   }
   {
    dst=Turn in Place
    rule=AnimGraph.AnimGraphNode_StateMachine_2.Locomotion.AnimStateTransitionNode_29.Transition
    src=Idle
    via=AnimStateTransitionNode_29
   }
   {
    dst=Idle
    rule=AnimGraph.AnimGraphNode_StateMachine_2.Locomotion.AnimStateTransitionNode_10.Transition
    src=JumpLand
    via=AnimStateTransitionNode_10
   }
   {
    dst=JumpLoop
    rule=AnimGraph.AnimGraphNode_StateMachine_2.Locomotion.AnimStateTransitionNode_27.Transition
    src=JumpLand
    via=AnimStateTransitionNode_27
   }
   {
    dst=Jumping
    rule=AnimGraph.AnimGraphNode_StateMachine_2.Locomotion.AnimStateTransitionNode_4.Transition
    src=JumpLand
    via=AnimStateTransitionNode_4
   }
   {
    dst=Movement
    rule=AnimGraph.AnimGraphNode_StateMachine_2.Locomotion.AnimStateTransitionNode_11.Transition
    src=JumpLand
    via=AnimStateTransitionNode_11
   }
   {
    dst=JumpLand
    rule=AnimGraph.AnimGraphNode_StateMachine_2.Locomotion.AnimStateTransitionNode_9.Transition
    src=JumpLoop
    via=AnimStateTransitionNode_9
   }
   {
    dst=JumpLoop
    rule=AnimGraph.AnimGraphNode_StateMachine_2.Locomotion.AnimStateTransitionNode_8.Transition
    src=Jumping
    via=AnimStateTransitionNode_8
   }
   {
    dst=Idle
    rule=AnimGraph.AnimGraphNode_StateMachine_2.Locomotion.AnimStateTransitionNode_1.Transition
    src=Movement
    via=AnimStateTransitionNode_1
   }
   {
    dst=Jumping
    rule=AnimGraph.AnimGraphNode_StateMachine_2.Locomotion.AnimStateTransitionNode_13.Transition
    src=Movement
    via=AnimStateTransitionNode_13
   }
   {
    dst=Movement
    rule=AnimGraph.AnimGraphNode_StateMachine_2.Locomotion.AnimStateTransitionNode_12.Transition
    src=Startled
    via=AnimStateTransitionNode_12
   }
   {
    dst=Swimming
    rule=AnimGraph.AnimGraphNode_StateMachine_2.Locomotion.AnimStateTransitionNode_6.Transition
    src=Swim State Alias
    via=AnimStateTransitionNode_6
   }
   {
    dst=Jumping
    rule=AnimGraph.AnimGraphNode_StateMachine_2.Locomotion.AnimStateTransitionNode_17.Transition
    src=Swimming
    via=AnimStateTransitionNode_17
   }
   {
    dst=Movement
    rule=AnimGraph.AnimGraphNode_StateMachine_2.Locomotion.AnimStateTransitionNode_16.Transition
    src=Swimming
    via=AnimStateTransitionNode_16
   }
   {
    dst=Topple_Idle
    rule=AnimGraph.AnimGraphNode_StateMachine_2.Locomotion.AnimStateTransitionNode_19.Transition
    src=Topple Alias
    via=AnimStateTransitionNode_19
   }
   {
    dst=Topple_Idle
    rule=AnimGraph.AnimGraphNode_StateMachine_2.Locomotion.AnimStateTransitionNode_7.Transition
    src=Topple Alias
    via=AnimStateTransitionNode_7
   }
   {
    dst=JumpLoop
    rule=AnimGraph.AnimGraphNode_StateMachine_2.Locomotion.AnimStateTransitionNode_20.Transition
    src=Topple_Idle
    via=AnimStateTransitionNode_20
   }
   {
    dst=Startled
    rule=AnimGraph.AnimGraphNode_StateMachine_2.Locomotion.AnimStateTransitionNode_2.Transition
    src=Topple_Idle
    via=AnimStateTransitionNode_2
   }
   {
    dst=Idle
    rule=AnimGraph.AnimGraphNode_StateMachine_2.Locomotion.AnimStateTransitionNode_30.Transition
    src=Turn in Place
    via=AnimStateTransitionNode_30
   }
  ]
 }
 Turn in Place {
  path=AnimGraph.AnimGraphNode_StateMachine_2.Locomotion.AnimStateNode_0.Turn in Place
  nodes [Output Animation Pose(Result=Sequence Player 'Labrador_Walk_N'() {groupName=Locomotion, method=SyncGroup, playRate=1.2, sequence=&Labrador_Walk_N})]
 }
 Topple_Idle {
  path=AnimGraph.AnimGraphNode_StateMachine_2.Locomotion.AnimStateNode_1.Topple_Idle
  nodes [
   AnimGraphNode_StateMachine_0 = LabradorTopple()
   n1 = Quadruped Slope Warping(ComponentPose=Local To Component(LocalPose=@AnimGraphNode_StateMachine_0), Alpha=1) {
    bOnlyAffectSkeletonRootBone=true
    bUseMovementComponentFloorInfoInsteadOfTraces=true
    maxStepHeight=35
    alphaBoolBlend {blendOption=Linear}
    armsFeetDefinitions [
     {numBonesInLimb=2, fKFootBone {boneName=frontleg_end_l}, iKFootBone {boneName=frontleg_04_l}}
     {numBonesInLimb=2, fKFootBone {boneName=frontleg_end_r}, iKFootBone {boneName=frontleg_end_r}}
    ]
    armsIKFootRootBone {boneName=lower_body}
    armsRootBone {boneName=upper_body}
    feetDefinitions [
     {numBonesInLimb=2, fKFootBone {boneName=backleg_end_l}, iKFootBone {boneName=backleg_end_l}}
     {numBonesInLimb=2, fKFootBone {boneName=backleg_end_r}, iKFootBone {boneName=backleg_end_r}}
    ]
    iKFootRootBone {boneName=lower_body}
    pelvisBone {boneName=lower_body}
    skeletonRootBone {boneName=root}
   }
   Output Animation Pose(Result=Component To Local(ComponentPose=@n1))
  ]
 }
 LabradorTopple {
  entry=GroundToSitupIdle
  path=AnimGraph.AnimGraphNode_StateMachine_2.Locomotion.AnimStateNode_1.Topple_Idle.AnimGraphNode_StateMachine_0.LabradorTopple
  ids=id
  nodes [
   AnimStateNode_1 = GroundToSitupIdle()
   AnimStateNode_2 = SitUpIdle()
   AnimStateTransitionNode_1 = AnimStateTransitionNode() {crossfadeDuration=0.5, logicType=TLT_Inertialization}
  ]
  links [{
    dst=SitUpIdle
    rule=AnimGraph.AnimGraphNode_StateMachine_2.Locomotion.AnimStateNode_1.Topple_Idle.AnimGraphNode_StateMachine_0.LabradorTopple.AnimStateTransitionNode_1.Transition
    src=GroundToSitupIdle
    via=AnimStateTransitionNode_1
   }]
 }
 GroundToSitupIdle {
  path=AnimGraph.AnimGraphNode_StateMachine_2.Locomotion.AnimStateNode_1.Topple_Idle.AnimGraphNode_StateMachine_0.LabradorTopple.AnimStateNode_1.GroundToSitupIdle
  nodes [Output Animation Pose(Result=Sequence Player 'ToppleToDizzy_Anim'() {sequence=&ToppleToDizzy_Anim})]
 }
 SitUpIdle {
  path=AnimGraph.AnimGraphNode_StateMachine_2.Locomotion.AnimStateNode_1.Topple_Idle.AnimGraphNode_StateMachine_0.LabradorTopple.AnimStateNode_2.SitUpIdle
  nodes [Output Animation Pose(Result=Sequence Player 'Labrador_Dizzy'() {sequence=&Labrador_Dizzy})]
 }
 Transition {
  path=AnimGraph.AnimGraphNode_StateMachine_2.Locomotion.AnimStateNode_1.Topple_Idle.AnimGraphNode_StateMachine_0.LabradorTopple.AnimStateTransitionNode_1.Transition
  nodes [Result(bCanEnterTransition=Get bIsInterruptible())]
 }
 Jumping {
  path=AnimGraph.AnimGraphNode_StateMachine_2.Locomotion.AnimStateNode_2.Jumping
  nodes [Output Animation Pose(Result=Sequence Player 'Labrador_JumpUp'() {sequence=/Labrador/Animation/Labrador_JumpUp})]
 }
 Idle {
  path=AnimGraph.AnimGraphNode_StateMachine_2.Locomotion.AnimStateNode_4.Idle
  nodes [
   n1 = Random Sequence Player() {entries [
     {
      chanceToPlay=0.5
      maxLoopCount=7
      maxPlayRate=1
      minLoopCount=3
      minPlayRate=1
      sequence=&Labrador_Idle
      blendIn {blendOption=Linear, blendTime=0, customCurve=null}
     }
     {
      chanceToPlay=0.2
      maxLoopCount=1
      maxPlayRate=1
      minLoopCount=1
      minPlayRate=1
      sequence=&Labrador_Idle_Ear_Twitch
      blendIn {blendOption=Linear, blendTime=0, customCurve=null}
     }
     {
      chanceToPlay=0.1
      maxLoopCount=1
      maxPlayRate=1
      minLoopCount=1
      minPlayRate=1
      sequence=&Labrador_Idle_Head_Twitch
      blendIn {blendOption=Linear, blendTime=0, customCurve=null}
     }
     {
      chanceToPlay=0.1
      maxLoopCount=1
      maxPlayRate=1
      minLoopCount=1
      minPlayRate=1
      sequence=&Labrador_Idle_Tail_Twitch
      blendIn {blendOption=Linear, blendTime=0, customCurve=null}
     }
     {
      chanceToPlay=0.1
      maxLoopCount=1
      maxPlayRate=1
      minLoopCount=1
      minPlayRate=1
      sequence=&Labrador_Mouth_Twitch
      blendIn {blendOption=Linear, blendTime=0, customCurve=null}
     }
    ]}
   Output Animation Pose(Result=@n1)
  ]
 }
 JumpLoop {
  path=AnimGraph.AnimGraphNode_StateMachine_2.Locomotion.AnimStateNode_5.JumpLoop
  nodes [
   AnimGraphNode_StateMachine_0 = InAir() {is=title AnimGraphNode_StateMachine}
   Output Animation Pose(Result=@AnimGraphNode_StateMachine_0)
  ]
 }
 InAir {
  entry=FallingLoop
  path=AnimGraph.AnimGraphNode_StateMachine_2.Locomotion.AnimStateNode_5.JumpLoop.AnimGraphNode_StateMachine_0.InAir
  ids=id
  nodes [
   AnimStateNode_0 = FallingLoop() {is=title AnimStateNode}
   AnimStateNode_1 = SpinLoop() {is=title AnimStateNode}
   AnimStateTransitionNode_0 = AnimStateTransitionNode() {logicType=TLT_Inertialization}
  ]
  links [{
    dst=SpinLoop
    rule=AnimGraph.AnimGraphNode_StateMachine_2.Locomotion.AnimStateNode_5.JumpLoop.AnimGraphNode_StateMachine_0.InAir.AnimStateTransitionNode_0.Transition
    src=FallingLoop
    via=AnimStateTransitionNode_0
   }]
 }
 FallingLoop {
  path=AnimGraph.AnimGraphNode_StateMachine_2.Locomotion.AnimStateNode_5.JumpLoop.AnimGraphNode_StateMachine_0.InAir.AnimStateNode_0.FallingLoop
  nodes [Output Animation Pose(Result=Sequence Player 'Labrador_FallLoop'() {sequence=/Labrador/Animation/Labrador_FallLoop})]
 }
 SpinLoop {
  path=AnimGraph.AnimGraphNode_StateMachine_2.Locomotion.AnimStateNode_5.JumpLoop.AnimGraphNode_StateMachine_0.InAir.AnimStateNode_1.SpinLoop
  nodes [Output Animation Pose(Result=Sequence Player 'Labrador_Spin'() {playRate=1.5, sequence=/Labrador/Animation/Labrador_Spin})]
 }
 Transition {
  path=AnimGraph.AnimGraphNode_StateMachine_2.Locomotion.AnimStateNode_5.JumpLoop.AnimGraphNode_StateMachine_0.InAir.AnimStateTransitionNode_0.Transition
  nodes [Result(bCanEnterTransition=Get bHighVelocity())]
 }
 JumpLand {
  path=AnimGraph.AnimGraphNode_StateMachine_2.Locomotion.AnimStateNode_6.JumpLand
  nodes [Output Animation Pose(Result=Sequence Player 'Labrador_Land'() {sequence=/Labrador/Animation/Labrador_Land})]
 }
 Startled {
  path=AnimGraph.AnimGraphNode_StateMachine_2.Locomotion.AnimStateNode_7.Startled
  nodes [Output Animation Pose(Result=Sequence Player 'Labrador_Startle_Mounted'() {bLoopAnimation=false, sequence=&Labrador_Startle_Mounted})]
 }
 Movement {
  path=AnimGraph.AnimGraphNode_StateMachine_2.Locomotion.AnimStateNode_8.Movement
  nodes [
   n1 = Blend Poses by bool(
    BlendPose_0=Sequence Player 'Labrador_Run_Bounding'(PlayRate=1.4) {playRate=1.3, sequence=&Labrador_Run_Bounding}
    BlendPose_1=Sequence Player 'Labrador_Run_Bounding'(PlayRate=1.2) {playRate=1.3, sequence=&Labrador_Run_Bounding}
    BlendTime_0=0.6
    BlendTime_1=0.6
    bActiveValue=Get bTargetIsClose()
   ) {blendType=Linear, transitionType=Inertialization}
   n2 = Sequence Player 'Labrador_Walk_N'() {groupName=Locomotion, method=SyncGroup, playRate=1.2, sequence=&Labrador_Walk_N}
   n3 = Get bIsRunning()
   n4 = Weighted Look At - Bone: head(
    ComponentPose=Local To Component(LocalPose=Blend Poses by bool(BlendPose_0=@n1, BlendPose_1=@n2, BlendTime_0=0.5, BlendTime_1=0.7, bActiveValue=@n3) {blendType=Linear, transitionType=Inertialization})
    Alpha=Get LookAtAlpha()
    LookAtLocation=Get AimTargetLocation()
   ) {
    interpolationTime=1
    interpolationType=EaseInOutExponent3
    lookAtClamp=60
    alphaBoolBlend {blendOption=Linear}
    boneToModify {boneName=head}
    boneWeightsToApply [{weight=0.8}, {weight=0.1}, {weight=0.1}]
    lookAt_Axis {bInLocalSpace=false, axis {y=0}}
    lookUp_Axis {bInLocalSpace=false, axis {x=0}}
   }
   Output Animation Pose(Result=Component To Local(ComponentPose=@n4))
  ]
 }
 Swimming {
  path=AnimGraph.AnimGraphNode_StateMachine_2.Locomotion.AnimStateNode_9.Swimming
  nodes [Output Animation Pose(Result=Sequence Player 'Labrador_Swim_N'() {playRate=1.4, sequence=&Labrador_Swim_N})]
 }
 Transition {
  path=AnimGraph.AnimGraphNode_StateMachine_2.Locomotion.AnimStateTransitionNode_0.Transition
  nodes [Result(bCanEnterTransition=Get bIsJumping())]
 }
 Transition {
  path=AnimGraph.AnimGraphNode_StateMachine_2.Locomotion.AnimStateTransitionNode_1.Transition
  nodes [Result(bCanEnterTransition=Get bLocoToIdle())]
 }
 Transition {
  path=AnimGraph.AnimGraphNode_StateMachine_2.Locomotion.AnimStateTransitionNode_10.Transition
  nodes [Result(bCanEnterTransition=Get bIsInterruptible())]
 }
 Transition {
  path=AnimGraph.AnimGraphNode_StateMachine_2.Locomotion.AnimStateTransitionNode_11.Transition
  nodes [Result(bCanEnterTransition=Get bJumpLandToMovement() -> (bJumpLandToMovement:bool))]
 }
 Transition {
  path=AnimGraph.AnimGraphNode_StateMachine_2.Locomotion.AnimStateTransitionNode_12.Transition
  nodes [
   Result(bCanEnterTransition=Get bIsInterruptible())
   AND Boolean(A="float < float"(A="Get Relevant Anim Time Remaining Fraction (Startled)"(), B=0.9), B=false)
  ]
 }
 Transition {
  path=AnimGraph.AnimGraphNode_StateMachine_2.Locomotion.AnimStateTransitionNode_13.Transition
  nodes [Result(bCanEnterTransition=Get bIsJumping())]
 }
 Transition {
  path=AnimGraph.AnimGraphNode_StateMachine_2.Locomotion.AnimStateTransitionNode_16.Transition
  nodes [Result(bCanEnterTransition=NOT Boolean(A=Get bIsSurfaceSwimming()))]
 }
 Transition {
  path=AnimGraph.AnimGraphNode_StateMachine_2.Locomotion.AnimStateTransitionNode_17.Transition
  nodes [Result(bCanEnterTransition=Get bIsJumping())]
 }
 Transition {
  path=AnimGraph.AnimGraphNode_StateMachine_2.Locomotion.AnimStateTransitionNode_19.Transition
  nodes [Result(bCanEnterTransition=Get bIsToppled())]
 }
 Transition {
  path=AnimGraph.AnimGraphNode_StateMachine_2.Locomotion.AnimStateTransitionNode_2.Transition
  nodes [Result(bCanEnterTransition=NOT Boolean(A=Get bIsToppled()))]
 }
 Transition {
  path=AnimGraph.AnimGraphNode_StateMachine_2.Locomotion.AnimStateTransitionNode_20.Transition
  nodes [Result(bCanEnterTransition=Get bHighVelocity())]
 }
 Transition {
  path=AnimGraph.AnimGraphNode_StateMachine_2.Locomotion.AnimStateTransitionNode_26.Transition
  nodes [Result(bCanEnterTransition=Get bIdleToFalling() -> (bIdleToFalling:bool))]
 }
 Transition {
  path=AnimGraph.AnimGraphNode_StateMachine_2.Locomotion.AnimStateTransitionNode_27.Transition
  nodes [Result(bCanEnterTransition=Get bHighVelocity())]
 }
 Transition {
  path=AnimGraph.AnimGraphNode_StateMachine_2.Locomotion.AnimStateTransitionNode_29.Transition
  nodes [Result(bCanEnterTransition=Get bIsTurningInPlace())]
 }
 Transition {
  path=AnimGraph.AnimGraphNode_StateMachine_2.Locomotion.AnimStateTransitionNode_3.Transition
  nodes [Result(bCanEnterTransition=Get bIdleToLoco() -> (bIdleToLoco:bool))]
 }
 Transition {
  path=AnimGraph.AnimGraphNode_StateMachine_2.Locomotion.AnimStateTransitionNode_30.Transition
  nodes [Result(bCanEnterTransition=NOT Boolean(A=Get bIsTurningInPlace()))]
 }
 Transition {
  path=AnimGraph.AnimGraphNode_StateMachine_2.Locomotion.AnimStateTransitionNode_4.Transition
  nodes [Result(bCanEnterTransition=Get bIsJumping())]
 }
 Transition {
  path=AnimGraph.AnimGraphNode_StateMachine_2.Locomotion.AnimStateTransitionNode_6.Transition
  nodes [Result(bCanEnterTransition=Get bIsSurfaceSwimming())]
 }
 Transition {
  path=AnimGraph.AnimGraphNode_StateMachine_2.Locomotion.AnimStateTransitionNode_7.Transition
  nodes [Result(bCanEnterTransition=Get bToppledOnGround())]
 }
 Transition {
  path=AnimGraph.AnimGraphNode_StateMachine_2.Locomotion.AnimStateTransitionNode_8.Transition
  nodes [Result(bCanEnterTransition=Get bJumpStartToJumpLoop() -> (bJumpStartToJumpLoop:bool))]
 }
 Transition {
  path=AnimGraph.AnimGraphNode_StateMachine_2.Locomotion.AnimStateTransitionNode_9.Transition
  nodes [Result(bCanEnterTransition=Get bTransitionToLand() -> (bTransitionToLand:bool))]
 }
 LootDrop {
  entry=IdenityPose
  path=AnimGraph.AnimGraphNode_StateMachine_3.LootDrop
  ids=id
  nodes [
   AnimStateNode_0 = IdenityPose() {is=title AnimStateNode}
   AnimStateNode_1 = LoopDrop() {is=title AnimStateNode}
   AnimStateTransitionNode_0 = AnimStateTransitionNode() {crossfadeDuration=0.4, logicType=TLT_Inertialization}
   AnimStateTransitionNode_1 = AnimStateTransitionNode() {bAutomaticRuleBasedOnSequencePlayerInState=true, crossfadeDuration=0.4, logicType=TLT_Inertialization}
  ]
  links [
   {
    dst=LoopDrop
    rule=AnimGraph.AnimGraphNode_StateMachine_3.LootDrop.AnimStateTransitionNode_0.Transition
    src=IdenityPose
    via=AnimStateTransitionNode_0
   }
   {
    dst=IdenityPose
    rule=AnimGraph.AnimGraphNode_StateMachine_3.LootDrop.AnimStateTransitionNode_1.Transition
    src=LoopDrop
    via=AnimStateTransitionNode_1
   }
  ]
 }
 IdenityPose {
  path=AnimGraph.AnimGraphNode_StateMachine_3.LootDrop.AnimStateNode_0.IdenityPose
  nodes [Output Animation Pose(Result=Additive Identity Pose())]
 }
 LoopDrop {
  path=AnimGraph.AnimGraphNode_StateMachine_3.LootDrop.AnimStateNode_1.LoopDrop
  nodes [Output Animation Pose(Result="Sequence Player (additive) 'Labrador_HitReact_A'"() {bLoopAnimation=false, sequence=/Labrador/Animation/Labrador_HitReact_A})]
 }
 Transition {
  path=AnimGraph.AnimGraphNode_StateMachine_3.LootDrop.AnimStateTransitionNode_0.Transition
  nodes [Result(bCanEnterTransition=Get bIsLeaking() -> (bIsLeaking:bool))]
 }
 Transition {
  path=AnimGraph.AnimGraphNode_StateMachine_3.LootDrop.AnimStateTransitionNode_1.Transition
  nodes [Result(bCanEnterTransition=Get bIsInterruptible())]
 }
 EventGraph {nodes [
   AnimNotify_LandEntered() {is=title K2Node_Event}
   > Set bJustLanded(bJustLanded:bool=true) -> (Output_Get:bool)
   > Delay(Duration=1)
   > Set bJustLanded(bJustLanded=false)
   Event BlueprintPostEvaluateAnimation() {is=title K2Node_Event}
   > Set bIsLeaking(bIsLeaking:bool=false) -> (Output_Get:bool)
   n1 = Event BlueprintProcessAnimInputEvent()
   > n2 = Sequence()
   > Branch(Condition=IsNamed(self=@n1.AnimInputEvent, InEventName=aie.Labrador.Leak))
   > Set bIsLeaking(bIsLeaking=true)
   Branch(execute=@n2.then_1, Condition=IsNamed(self=@n1.AnimInputEvent, InEventName=aie.Labrador.IsToppled))
   > n3 = Cast To FortAnimInputEvent_Bool!(execute, Object=@n1.AnimInputEvent) -> (then, CastFailed, AsFort Anim Input Event Bool:Fort Anim Input Event Bool Object Reference) {is=title K2Node_DynamicCast}
   > Set bIsToppled(bIsToppled:bool=Get bValue!(self:Fort Anim Input Event Bool Object Reference=@n3) -> (bValue:bool)) -> (Output_Get:bool)
  ]}
]
variables []
