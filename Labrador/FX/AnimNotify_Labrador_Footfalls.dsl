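; Animation notify for the Labrador that triggers footfall dust on a wildlife NPC pawn.
; It casts the mesh owner to NPC_Pawn_Wildlife_Parent and, if that pawn's Wildlife_FootDust particle component is valid, activates it with a reset.
; It then passes the notify's running flag and footfall index to the effect as bIsRunning and SkeletalIndex.
; The mesh's right vector supplies the effect's ForwardVector parameter.
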(fn Received_Notify
  (params (MeshComp :type Skeletal%20Mesh%20Component%20Object%20Reference) (Animation :type Anim%20Sequence%20Base%20Object%20Reference) (EventReference :type Anim%20Notify%20Event%20Reference%20Structure%20%28by%20ref%29))
  (bind _asnpc_pawn_wildlife_parent (Utilities|Casting|CastToNPC_Pawn_Wildlife_Parent (Components|GetOwner MeshComp)))
  (Utilities|IsValid (Variables|/Game/Athena/AI/Pawns/NPC/Parents/NPC_Pawn_Wildlife_Parent.NPC_Pawn_Wildlife_Parent_C|GetWildlife_FootDust _asnpc_pawn_wildlife_parent)
    (:Is%20Valid
      (Components|Activation|Activate (Variables|/Game/Athena/AI/Pawns/NPC/Parents/NPC_Pawn_Wildlife_Parent.NPC_Pawn_Wildlife_Parent_C|GetWildlife_FootDust _asnpc_pawn_wildlife_parent) true)
      (Effects|Components|ParticleSystem|SetBoolParameter (Variables|/Game/Athena/AI/Pawns/NPC/Parents/NPC_Pawn_Wildlife_Parent.NPC_Pawn_Wildlife_Parent_C|GetWildlife_FootDust _asnpc_pawn_wildlife_parent) "bIsRunning" (Variables|Default|Getis%20Running%20?))
      (Effects|Components|ParticleSystem|SetVectorParameter (Variables|/Game/Athena/AI/Pawns/NPC/Parents/NPC_Pawn_Wildlife_Parent.NPC_Pawn_Wildlife_Parent_C|GetWildlife_FootDust _asnpc_pawn_wildlife_parent) "ForwardVector" (Transformation|GetRightVector MeshComp))
      (Effects|Components|ParticleSystem|SetIntParameter (Variables|/Game/Athena/AI/Pawns/NPC/Parents/NPC_Pawn_Wildlife_Parent.NPC_Pawn_Wildlife_Parent_C|GetWildlife_FootDust _asnpc_pawn_wildlife_parent) "SkeletalIndex" (Variables|Default|GetFootfallsIndex)))
    (:Is%20Not%20Valid
      _))
  (return false))
