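# Niagara effect for the Labrador that, going by its name, signals anticipation of completed training, using one mesh particle attached to the dog.
# A single burst spawns the particle, whose lifetime loops instead of killing it while alpha fades over each cycle.
# Every update samples the User.LabradorMesh skeleton for position and slowly rotates the mesh.
# Color cycles through a rainbow curve driven by engine time, independent of particle age.
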NiagaraSystem /Labrador/FX/NS_Labrador_AnticipateTrainingCompleted
dependencies {
 usedDataInterfaces [&NiagaraDataInterfaceColorCurve, &NiagaraDataInterfaceCurve]
 usedDynamicInputs []
 usedModules [&SystemState]
}
emitters {PickUp_Poop {
  emitterUpdateScript [
   &EmitterState()
   &SpawnBurst_Instantaneous(Spawn Count=1, Use Spawn Probability=false)
  ]
  particleSpawnScript [
   &InitializeParticle(
    Color=<rgba 2 2 2 1>
    Lifetime=10
    Mesh Scale Mode=::NewEnumerator3
    Mesh Uniform Scale=100
    Position Offset=<0 0 45>
    UsePositionOffset=true
    Write Material Random=true
   )
   &DynamicMaterialParameters(
    Index 0 Param 1=&RandomRangeFloat(Minimum=0, Maximum=5)
    Index 0 Param 2=5
    Index 0 Param 3=1
    Index 0 Param 4=1
    Module.Param0WriteEnabled=true
    Module.Param1WriteEnabled=true
    Module.Param2WriteEnabled=true
    Module.Param3WriteEnabled=false
   )
   &CameraOffset(Camera Offset Amount=100)
  ]
  particleUpdateScript [
   &ParticleState(Kill Particles When Lifetime Has Elapsed=false, Loop Particles Lifetime=true)
   &ScaleColor(
    Scale Alpha=&FloatFromCurve(FloatCurve=DataInterface() {propertyValues {Curve {keys [<at 0 1 RCIM_Cubic>, {arriveTangent=-3.3542, interpMode=RCIM_Cubic, leaveTangent=-3.3542, tangentMode=RCTM_User, time=1, value=0}]}}})
    Scale RGB=&MakeVectorFromLinearColorRGB(Color=&ColorFromCurve(
      ColorCurve=DataInterface() {propertyValues {
        AlphaCurve {keys [<at 0 1 RCIM_Cubic>,<at 1 1 RCIM_Cubic>]}
        BlueCurve {keys [
          <at 0 0.6514 RCIM_Cubic>
          {arriveTangent=0.6972, interpMode=RCIM_Cubic, leaveTangent=0.6972, time=0.25, value=0.9216}
          {arriveTangent=-0.8685, interpMode=RCIM_Cubic, leaveTangent=-0.8685, time=0.5, value=1}
          {arriveTangent=-0.6983, interpMode=RCIM_Cubic, leaveTangent=-0.6983, time=0.75, value=0}
          <at 0.9992 0.6514 RCIM_Cubic>
         ]}
        GreenCurve {keys [
          <at 0 0.0497 RCIM_Cubic>
          {arriveTangent=1.83, interpMode=RCIM_Cubic, leaveTangent=1.83, time=0.25, value=0.0762}
          {arriveTangent=1.2671, interpMode=RCIM_Cubic, leaveTangent=1.2671, time=0.5, value=0.9647}
          {arriveTangent=-1.8329, interpMode=RCIM_Cubic, leaveTangent=-1.8329, time=0.75, value=1}
          <at 0.9992 0.0497 RCIM_Cubic>
         ]}
        RedCurve {keys [
          <at 0 1 RCIM_Cubic>
          {arriveTangent=-2, interpMode=RCIM_Cubic, leaveTangent=-2, time=0.25, value=0.013}
          <at 0.5 0 RCIM_Cubic>
          {arriveTangent=2.0033, interpMode=RCIM_Cubic, leaveTangent=2.0033, time=0.75, value=1}
          <at 0.9992 1 RCIM_Cubic>
         ]}
       }}
      CurveIndex=&Frac_Float(Float=&Multiply_Float(A=$Engine.Time, B=2))
     ))
    ScaleA=true
    ScaleRGB=true
   )
   &UpdateMeshOrientation(Rotation Rate=0.1, Rotation Vector=<1 0 0>)
   &SkeletalMeshLocation(Bone Sampling Mode=::NewEnumerator4, MaskSpawnGroup=false, Sampled Position Offset=<0 0 10>, Skeletal Mesh=$User.LabradorMesh)
  ]
  rendererClasses [&NiagaraMeshRendererProperties]
  renderers [{rendererClass=&NiagaraMeshRendererProperties, rendererIndex=0}]
 }}
summary {
 systemName=NS_Labrador_AnticipateTrainingCompleted
 emitters [{emitterName=PickUp_Poop, rendererClasses [&NiagaraMeshRendererProperties]}]
 userVariables [User.LabradorMesh {
   type=NiagaraDataInterfaceSkeletalMesh
   defaultValue {
    struct=NiagaraExt_VariableValue_DataInterface
    value {dataInterface=&self:NiagaraDataInterfaceSkeletalMesh_0}
   }
  }]
}
