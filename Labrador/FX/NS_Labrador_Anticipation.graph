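# Niagara anticipation effect for a Labrador that builds up around its skeletal mesh while a poop pickup mesh fades in and grows.
# GPU beam sprites and mesh fragments ramp up on User.LabradorMesh, alongside GPU confetti, a tapering shimmer and a light.
# Beams, shimmer, light and pickup share a time-driven rainbow cycle, and User.SystemLife sets pickup lifetime.
# Owner LOD distance enlarges beams and shimmer and culls confetti, fragments and light.
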NiagaraSystem /Labrador/FX/NS_Labrador_Anticipation
dependencies {
 usedDataInterfaces [&NiagaraDataInterfaceCurve, &NiagaraDataInterfaceColorCurve]
 usedDynamicInputs []
 usedModules [&SystemState]
}
emitters {
 Beam_Up_Build_Up {
  simTarget=GPUComputeSim
  emitterUpdateScript [
   &EmitterState()
   &SpawnRate(
    SpawnRate=&FloatFromCurve(
     FloatCurve=DataInterface() {propertyValues {Curve {keys [<at 0 0 RCIM_Cubic>, {arriveTangent=2.9325, interpMode=RCIM_Cubic, leaveTangent=2.9325, tangentMode=RCTM_User, time=1, value=1}]}}}
     CurveIndex=$Emitter.NormalizedLoopAge
     Scale Curve=75
    )
    Use Spawn Probability=false
   )
  ]
  particleSpawnScript [&InitializeParticle(
    Lifetime=&RandomRangeFloat(Minimum=0.25, Maximum=0.45)
    Sprite Size Mode=::NewEnumerator1
    Uniform Sprite Size Max=100
    Uniform Sprite Size Min=50
    UsePositionOffset=false
    Write Material Random=true
   )]
  particleUpdateScript [
   &ParticleState()
   &SkeletalMeshLocation(
    Bone Sampling Mode=::NewEnumerator0
    MaskSpawnGroup=false
    OverrideRandomSeed=false
    Position Sampling=::NewEnumerator1
    Sampled Position Offset=&RandomRangeVector(Minimum=<-20 -20 10>, Maximum=<20 20 20>)
    Skeletal Mesh=$User.LabradorMesh
   )
   /Niagara/Modules/Math/PlaceParticlesOnDepthBuffer_GPU(
    Max Z Depth=50
    Module.Use Max Z Depth=true
    Write To Intrinsic Variables:NiagaraBool=&SetBoolByFloatComparison(A=$Particles.NormalizedAge, B=0.7, Comparison Type=::NewEnumerator7)
   )
   &AccelerationForce(Acceleration=&Multiply_VectorByFloat(
     Vector=&RandomRangeVector(Minimum=<0 0 1000>, Maximum=<0 0 1000>)
     Float=&FloatFromCurve(
      FloatCurve=DataInterface() {propertyValues {Curve {keys [
          {arriveTangent=2.9438, interpMode=RCIM_Cubic, time=0, value=0}
          {arriveTangent=2.448, interpMode=RCIM_Cubic, leaveTangent=2.448, tangentMode=RCTM_User, time=1, value=1}
         ]}}}
      Scale Curve=10
     )
    ))
   &SolveForcesAndVelocity(Clamp Velocity=false, Limit Acceleration=false)
   &ScaleSpriteSize(Scale Factor=&MakeVector2D(
     X=&FloatFromCurve(FloatCurve=DataInterface() {propertyValues {Curve {keys [<at 0 1>,<at 1 0>]}}})
     Y=&FloatFromCurve(
      FloatCurve=DataInterface() {propertyValues {Curve {keys [<at 0.5 1 RCIM_Cubic>, {arriveTangent=29.9912, interpMode=RCIM_Cubic, leaveTangent=29.9912, tangentMode=RCTM_User, time=1, value=10}]}}}
      Scale Curve=4
     )
    ))
   &ScaleSpriteSize(Scale Factor=&Vector2DFromFloat(Value=&RemapRange(Input Value To Remap=$Engine.Owner.LODDistance, Input Min=0, Input Max=7500, Output Min=1, Output Max=5))) {module=ScaleSpriteSize001}
   &CameraOffset(Camera Offset Amount=&FloatFromCurve(FloatCurve=DataInterface() {propertyValues {Curve {keys [<at 0 1>,<at 1 5>]}}}, Scale Curve=10))
   &ScaleColor(
    Scale Alpha=&FloatFromCurve(FloatCurve=DataInterface() {propertyValues {Curve {keys [<at 0.5 1 RCIM_Cubic>, {arriveTangent=-5.4338, interpMode=RCIM_Cubic, leaveTangent=-5.4338, tangentMode=RCTM_User, time=1, value=0}]}}})
    Scale RGB=&MakeVectorFromLinearColorRGB(Color=&ColorFromCurve(
      ColorCurve=DataInterface() {propertyValues {
        AlphaCurve {keys [<at 0 1 RCIM_Cubic>,<at 1 1 RCIM_Cubic>]}
        BlueCurve {keys [
          <at 0 0.6514 RCIM_Cubic>
          {arriveTangent=0.6972, interpMode=RCIM_Cubic, leaveTangent=0.6972, time=0.25, value=0.9216}
          {arriveTangent=-0.8685, interpMode=RCIM_Cubic, leaveTangent=-0.8685, time=0.5, value=1}
          {arriveTangent=-0.6983, interpMode=RCIM_Cubic, leaveTangent=-0.6983, time=0.75, value=0}
          <at 0.9992 0.6514 RCIM_Cubic>
         ]}
        GreenCurve {keys [
          <at 0 0.0497 RCIM_Cubic>
          {arriveTangent=1.83, interpMode=RCIM_Cubic, leaveTangent=1.83, time=0.25, value=0.0762}
          {arriveTangent=1.2671, interpMode=RCIM_Cubic, leaveTangent=1.2671, time=0.5, value=0.9647}
          {arriveTangent=-1.8329, interpMode=RCIM_Cubic, leaveTangent=-1.8329, time=0.75, value=1}
          <at 0.9992 0.0497 RCIM_Cubic>
         ]}
        RedCurve {keys [
          <at 0 1 RCIM_Cubic>
          {arriveTangent=-2, interpMode=RCIM_Cubic, leaveTangent=-2, time=0.25, value=0.013}
          <at 0.5 0 RCIM_Cubic>
          {arriveTangent=2.0033, interpMode=RCIM_Cubic, leaveTangent=2.0033, time=0.75, value=1}
          <at 0.9992 1 RCIM_Cubic>
         ]}
       }}
      CurveIndex=&Frac_Float(Float=&Multiply_Float(A=$Engine.Time, B=2))
     ))
    ScaleA=true
    ScaleRGB=true
   )
  ]
  rendererClasses [&NiagaraSpriteRendererProperties]
  renderers [{rendererClass=&NiagaraSpriteRendererProperties}]
 }
 Confetti {
  simTarget=GPUComputeSim
  emitterUpdateScript [
   &EmitterState()
   &SpawnRate(SpawnRate=30, Use Spawn Probability=false)
  ]
  particleSpawnScript [
   &InitializeParticle(
    AdjustAlpha=false
    AdjustHue=false
    AdjustSaturation=false
    AdjustValue=false
    Color=&RandomRangeLinearColor(Minimum=<rgba 0.5 0.305 0 1>, Maximum=<rgba 0.75 0.305 0 1>)
    Color Mode=::NewEnumerator3
    Lifetime Max=1.5
    Lifetime Min=0.5
    Lifetime Mode=::NewEnumerator1
    Sprite Rotation Mode=::NewEnumerator2
    Sprite Size Mode=::NewEnumerator1
    Uniform Sprite Size Max=7
    Uniform Sprite Size Min=3
    UsePositionOffset=false
    Write Material Random=true
   )
   &SkeletalMeshLocation(
    Bone Sampling Mode=::NewEnumerator0
    Inherit Velocity=false
    MaskSpawnGroup=false
    Module.Override Randomness=false
    Skeletal Mesh=$User.LabradorMesh
   )
   &SphereLocation(
    MaskSpawnGroup=false
    Override Seed=false
    Sphere Radius=&RandomRangeFloat(Minimum=15, Maximum=25)
    Surface Only=false
    UseNonUniformScale=false
    UseOffset=false
    UseSphereOrientation=false
   )
   &AddVelocityFromPoint(Module.Use Falloff Distance=false, Velocity Strength=&RandomRangeFloat(Minimum=50, Maximum=100))
  ]
  particleUpdateScript [
   &ParticleState()
   &AccelerationForce(Acceleration=<0 0 50>)
   &Drag(Drag=&RandomRangeFloat(Minimum=1, Maximum=3), Use Linear Drag=true, Use Rotational Drag=false)
   &CurlNoiseForce(Noise Frequency=10, Noise Strength=200, Pan Noise=true, Pan Noise Field=<0 0 10>)
   &SolveForcesAndVelocity(Clamp Velocity=false, Limit Acceleration=false)
   &ScaleSpriteSize(Scale Factor=&Vector2DFromFloat(Value=&FloatFromCurve(FloatCurve=DataInterface() {propertyValues {Curve {keys [
          {arriveTangent=-0.4596, interpMode=RCIM_Cubic, leaveTangent=-0.4596, tangentMode=RCTM_User, time=0.55, value=1}
          {arriveTangent=-4.9031, interpMode=RCIM_Cubic, leaveTangent=-4.9031, tangentMode=RCTM_User, time=1, value=0}
         ]}}})))
   &ScaleColor(
    Scale RGB=&MakeVector(
     X=1
     Y=&Add_Float(
      A=&RandomRangeFloat(Minimum=0, Maximum=1000)
      B=&Lerp_Float(A=0, B=&RandomRangeFloat(Minimum=25, Maximum=150), Alpha=$Particles.NormalizedAge)
     )
     Z=1
    )
    ScaleA=false
    ScaleRGB=true
   )
   &KillParticles(Kill Particles=&SetBoolByFloatComparison(A=$Engine.Owner.LODDistance, B=7500))
  ]
  rendererClasses [&NiagaraSpriteRendererProperties]
  renderers [{rendererClass=&NiagaraSpriteRendererProperties}]
 }
 Light {
  emitterUpdateScript [
   &EmitterState()
   &SpawnBurst_Instantaneous(Spawn Count=1, Use Spawn Probability=false)
  ]
  particleSpawnScript [&InitializeParticle(Color Mode=::NewEnumerator0, Lifetime=5, UsePositionOffset=false, Write Material Random=true)]
  particleUpdateScript [
   &ParticleState(Kill Particles When Lifetime Has Elapsed=false, Let Infinitely Lived Particles Die When Emitter Deactivates=true)
   &FN_Light_Attributes(
    Color=&ColorFromCurve(
     ColorCurve=DataInterface() {propertyValues {
       AlphaCurve {keys [<at 0 1 RCIM_Cubic>,<at 1 1 RCIM_Cubic>]}
       BlueCurve {keys [
         <at 0 0.6514 RCIM_Cubic>
         {arriveTangent=0.6972, interpMode=RCIM_Cubic, leaveTangent=0.6972, time=0.25, value=0.9216}
         {arriveTangent=-0.8685, interpMode=RCIM_Cubic, leaveTangent=-0.8685, time=0.5, value=1}
         {arriveTangent=-0.6983, interpMode=RCIM_Cubic, leaveTangent=-0.6983, time=0.75, value=0}
         <at 0.9992 0.6514 RCIM_Cubic>
        ]}
       GreenCurve {keys [
         <at 0 0.0497 RCIM_Cubic>
         {arriveTangent=1.83, interpMode=RCIM_Cubic, leaveTangent=1.83, time=0.25, value=0.0762}
         {arriveTangent=1.2671, interpMode=RCIM_Cubic, leaveTangent=1.2671, time=0.5, value=0.9647}
         {arriveTangent=-1.8329, interpMode=RCIM_Cubic, leaveTangent=-1.8329, time=0.75, value=1}
         <at 0.9992 0.0497 RCIM_Cubic>
        ]}
       RedCurve {keys [
         <at 0 1 RCIM_Cubic>
         {arriveTangent=-2, interpMode=RCIM_Cubic, leaveTangent=-2, time=0.25, value=0.013}
         <at 0.5 0 RCIM_Cubic>
         {arriveTangent=2.0033, interpMode=RCIM_Cubic, leaveTangent=2.0033, time=0.75, value=1}
         <at 0.9992 1 RCIM_Cubic>
        ]}
      }}
     CurveIndex=&Frac_Float(Float=&Multiply_Float(A=$Engine.Time, B=2))
    )
    Create Color Attribute=true
    Create Exponent Attribute=true
    Create Position Attribute=true
    Create Radius Attribute=true
    Exponent=3
    Radius=100
    Scale Color=20
   )
   &KillParticles(Kill Particles=&SetBoolByFloatComparison(A=$Engine.Owner.LODDistance, B=7500))
  ]
  rendererClasses [&NiagaraLightRendererProperties]
  renderers [{rendererClass=&NiagaraLightRendererProperties, rendererIndex=0}]
 }
 PickUp_Poop {
  emitterUpdateScript [
   &EmitterState(Life Cycle Mode=::NewEnumerator1, Loop Behavior=::NewEnumerator1, Loop Delay=0.05, Loop Duration=1, UseLoopDelay=true)
   &SpawnBurst_Instantaneous(Spawn Count=1, Use Spawn Probability=false)
  ]
  particleSpawnScript [
   &InitializeParticle(
    Color=<rgba 2 2 2 1>
    Lifetime=&Add_Float(A=$User.SystemLife, B=0.75)
    Mesh Scale Mode=::NewEnumerator3
    Mesh Uniform Scale=100
    UsePositionOffset=false
    Write Material Random=true
   )
   &DynamicMaterialParameters(
    Index 0 Param 1=&RandomRangeFloat(Minimum=0, Maximum=5)
    Index 0 Param 2=5
    Index 0 Param 3=1
    Index 0 Param 4=1
    Module.Param0WriteEnabled=true
    Module.Param1WriteEnabled=true
    Module.Param2WriteEnabled=true
    Module.Param3WriteEnabled=false
   )
  ]
  particleUpdateScript [
   &ParticleState()
   &CameraOffset(Camera Offset Amount=&FloatFromCurve(FloatCurve=DataInterface() {propertyValues {Curve {keys [<at 0 0 RCIM_Cubic>, {arriveTangentWeight=0.3333, interpMode=RCIM_Cubic, tangentMode=RCTM_User, time=1, value=100}]}}}))
   &ScaleColor(
    Scale Alpha=&FloatFromCurve(FloatCurve=DataInterface() {propertyValues {Curve {keys [{arriveTangent=3.032, interpMode=RCIM_Cubic, leaveTangent=3.032, tangentMode=RCTM_User, time=0, value=0}, <at 1 1 RCIM_Cubic>]}}})
    Scale RGB=&MakeVectorFromLinearColorRGB(Color=&ColorFromCurve(
      ColorCurve=DataInterface() {propertyValues {
        AlphaCurve {keys [<at 0 1 RCIM_Cubic>,<at 1 1 RCIM_Cubic>]}
        BlueCurve {keys [
          <at 0 0.6514 RCIM_Cubic>
          {arriveTangent=0.6972, interpMode=RCIM_Cubic, leaveTangent=0.6972, time=0.25, value=0.9216}
          {arriveTangent=-0.8685, interpMode=RCIM_Cubic, leaveTangent=-0.8685, time=0.5, value=1}
          {arriveTangent=-0.6983, interpMode=RCIM_Cubic, leaveTangent=-0.6983, time=0.75, value=0}
          <at 0.9992 0.6514 RCIM_Cubic>
         ]}
        GreenCurve {keys [
          <at 0 0.0497 RCIM_Cubic>
          {arriveTangent=1.83, interpMode=RCIM_Cubic, leaveTangent=1.83, time=0.25, value=0.0762}
          {arriveTangent=1.2671, interpMode=RCIM_Cubic, leaveTangent=1.2671, time=0.5, value=0.9647}
          {arriveTangent=-1.8329, interpMode=RCIM_Cubic, leaveTangent=-1.8329, time=0.75, value=1}
          <at 0.9992 0.0497 RCIM_Cubic>
         ]}
        RedCurve {keys [
          <at 0 1 RCIM_Cubic>
          {arriveTangent=-2, interpMode=RCIM_Cubic, leaveTangent=-2, time=0.25, value=0.013}
          <at 0.5 0 RCIM_Cubic>
          {arriveTangent=2.0033, interpMode=RCIM_Cubic, leaveTangent=2.0033, time=0.75, value=1}
          <at 0.9992 1 RCIM_Cubic>
         ]}
       }}
      CurveIndex=&Frac_Float(Float=&Multiply_Float(A=$Engine.Time, B=2))
     ))
    ScaleA=true
    ScaleRGB=true
   )
   &ScaleMeshSize(Scale Factor=&VectorFromFloat(Value=&FloatFromCurve(FloatCurve=DataInterface() {propertyValues {Curve {keys [{arriveTangent=2.0756, interpMode=RCIM_Cubic, leaveTangent=2.0756, tangentMode=RCTM_User, time=0, value=1}, <at 1 2 RCIM_Cubic>]}}})))
   &UpdateMeshOrientation(Rotation Rate=0.1, Rotation Vector=<1 0 0>)
  ]
  rendererClasses [&NiagaraMeshRendererProperties]
  renderers [{rendererClass=&NiagaraMeshRendererProperties, rendererIndex=0}]
 }
 Shimmer {
  emitterUpdateScript [
   &EmitterState()
   &SpawnRate(
    SpawnRate=&FloatFromCurve(
     FloatCurve=DataInterface() {propertyValues {Curve {keys [<at 0 1 RCIM_Cubic>, {arriveTangent=-3.2315, interpMode=RCIM_Cubic, leaveTangent=-3.2315, tangentMode=RCTM_User, time=1, value=0}]}}}
     CurveIndex=$Emitter.NormalizedLoopAge
     Scale Curve=5
    )
    Use Spawn Probability=false
   )
  ]
  particleSpawnScript [
   &InitializeParticle(
    Color=&MakeLinearColorFromVectorAndFloat(
     "Vector (RGB)"=&MakeVectorFromLinearColorRGB(Color=<rgba 0.75 0.75 0.75 1>)
     "Float (Alpha)"=&RandomRangeFloat(Minimum=0.2, Maximum=0.5)
    )
    Lifetime Mode=::NewEnumerator1
    Sprite Rotation Mode=::NewEnumerator2
    Sprite Size Mode=::NewEnumerator1
    Uniform Sprite Size Max=250
    Uniform Sprite Size Min=225
    UsePositionOffset=false
    Write Material Random=true
   )
   &SkeletalMeshLocation(Bone Sampling Mode=::NewEnumerator0, MaskSpawnGroup=false, OverrideRandomSeed=false, Skeletal Mesh=$User.LabradorMesh) {enabled=false}
   &SphereLocation(
    MaskSpawnGroup=false
    Override Seed=false
    Sphere Radius=25
    Surface Only=false
    UseNonUniformScale=false
    UseOffset=false
    UseSphereOrientation=false
   )
   &DynamicMaterialParameters(
    Index 0 Param 1=&RandomRangeFloat(Minimum=0, Maximum=1)
    Index 0 Param 2=&RandomRangeFloat(Minimum=0, Maximum=1)
    Module.Param0WriteEnabled=true
    Module.Param1WriteEnabled=true
    Module.Param2WriteEnabled=true
    Module.Param3WriteEnabled=true
   )
  ]
  particleUpdateScript [
   &ParticleState()
   &ScaleColor(
    Scale Alpha=&FloatFromCurve(FloatCurve=DataInterface() {propertyValues {Curve {keys [
         {interpMode=RCIM_Cubic, tangentMode=RCTM_User, time=0, value=0}
         {interpMode=RCIM_Cubic, tangentMode=RCTM_User, time=0.5, value=1}
         {tangentMode=RCTM_User, time=1, value=0}
        ]}}})
    Scale RGB=&MakeVectorFromLinearColorRGB(Color=&ColorFromCurve(
      ColorCurve=DataInterface() {propertyValues {
        AlphaCurve {keys [<at 0 1 RCIM_Cubic>,<at 1 1 RCIM_Cubic>]}
        BlueCurve {keys [
          <at 0 0.6514 RCIM_Cubic>
          {arriveTangent=0.6972, interpMode=RCIM_Cubic, leaveTangent=0.6972, time=0.25, value=0.9216}
          {arriveTangent=-0.8685, interpMode=RCIM_Cubic, leaveTangent=-0.8685, time=0.5, value=1}
          {arriveTangent=-0.6983, interpMode=RCIM_Cubic, leaveTangent=-0.6983, time=0.75, value=0}
          <at 0.9992 0.6514 RCIM_Cubic>
         ]}
        GreenCurve {keys [
          <at 0 0.0497 RCIM_Cubic>
          {arriveTangent=1.83, interpMode=RCIM_Cubic, leaveTangent=1.83, time=0.25, value=0.0762}
          {arriveTangent=1.2671, interpMode=RCIM_Cubic, leaveTangent=1.2671, time=0.5, value=0.9647}
          {arriveTangent=-1.8329, interpMode=RCIM_Cubic, leaveTangent=-1.8329, time=0.75, value=1}
          <at 0.9992 0.0497 RCIM_Cubic>
         ]}
        RedCurve {keys [
          <at 0 1 RCIM_Cubic>
          {arriveTangent=-2, interpMode=RCIM_Cubic, leaveTangent=-2, time=0.25, value=0.013}
          <at 0.5 0 RCIM_Cubic>
          {arriveTangent=2.0033, interpMode=RCIM_Cubic, leaveTangent=2.0033, time=0.75, value=1}
          <at 0.9992 1 RCIM_Cubic>
         ]}
       }}
      CurveIndex=&Frac_Float(Float=&Multiply_Float(A=$Engine.Time, B=2))
     ))
    ScaleA=true
    ScaleRGB=true
   )
   &CameraOffset(Camera Offset Amount=&FloatFromCurve(FloatCurve=DataInterface() {propertyValues {Curve {keys [<at 0 -1>,<at 1 1>]}}}, Scale Curve=50))
   &ScaleSpriteSize(Scale Factor=&Vector2DFromFloat(Value=&RemapRange(Input Value To Remap=$Engine.Owner.LODDistance, Input Min=0, Input Max=7500, Output Min=1, Output Max=2)))
  ]
  rendererClasses [&NiagaraSpriteRendererProperties]
  renderers [{rendererClass=&NiagaraSpriteRendererProperties}]
 }
 TinyFrags {
  emitterUpdateScript [
   &EmitterState()
   &SpawnRate(
    SpawnRate=&FloatFromCurve(
     FloatCurve=DataInterface() {propertyValues {Curve {keys [
         {arriveTangent=2.7721, interpMode=RCIM_Cubic, leaveTangent=2.7721, tangentMode=RCTM_User, time=0, value=0}
         {interpMode=RCIM_Cubic, leaveTangent=-1.7992, time=1, value=1}
        ]}}}
     CurveIndex=$Emitter.NormalizedLoopAge
     Scale Curve=25
    )
    Use Spawn Probability=false
   )
  ]
  particleSpawnScript [
   &InitializeParticle(
    Color=<rgba 10 10 10 1>
    Lifetime Max=1.25
    Lifetime Mode=::NewEnumerator1
    Mesh Scale Max=<0.7 0.7 1>
    Mesh Scale Min=<0.4 0.4 1>
    Mesh Scale Mode=::NewEnumerator2
    UsePositionOffset=false
    Write Material Random=true
   )
   &SkeletalMeshLocation(
    Bone Sampling Mode=::NewEnumerator0
    MaskSpawnGroup=false
    OverrideRandomSeed=false
    Position Sampling=::NewEnumerator1
    Skeletal Mesh=$User.LabradorMesh
   )
   &SphereLocation(
    MaskSpawnGroup=false
    Override Seed=false
    Sphere Radius=25
    Surface Only=false
    UseNonUniformScale=false
    UseOffset=false
    UseSphereOrientation=false
   )
   &InitialMeshOrientation(
    Rotation=&RandomRangeVector(Minimum=<0 0 0>, Maximum=<1 1 1>)
    Use Orientation Vector=false
    Use Rotation Vector=true
   )
   &AddVelocityFromPoint(Module.Use Falloff Distance=false, Velocity Strength=&RandomRangeFloat(Minimum=50, Maximum=100))
  ]
  particleUpdateScript [
   &ParticleState()
   &AccelerationForce(Acceleration=<0 0 50>)
   &Drag(Drag=&RandomRangeFloat(Minimum=1, Maximum=3), Use Linear Drag=true, Use Rotational Drag=false)
   &CurlNoiseForce(Noise Frequency=10, Noise Strength=200, Pan Noise=true, Pan Noise Field=<0 0 10>)
   &SolveForcesAndVelocity(Clamp Velocity=false, Limit Acceleration=false)
   &UpdateMeshOrientation(Rotation Rate=0.15, Rotation Vector=&RandomRangeVector(Minimum=<0 0 0>, Maximum=<1 1 1>))
   &ScaleMeshSize(Scale Factor=&VectorFromFloat(Value=&FloatFromCurve(FloatCurve=DataInterface() {propertyValues {Curve {keys [
          {arriveTangent=12.2517, interpMode=RCIM_Cubic, leaveTangent=12.2517, tangentMode=RCTM_User, time=0, value=0}
          {arriveTangentWeight=0.1067, interpMode=RCIM_Cubic, leaveTangentWeight=0.3202, tangentMode=RCTM_User, time=0.2, value=1}
          {arriveTangentWeight=0.3202, interpMode=RCIM_Cubic, leaveTangentWeight=0.1067, tangentMode=RCTM_User, time=0.8, value=1}
          <at 1 0 RCIM_Cubic>
         ]}}})))
   &KillParticles(Kill Particles=&SetBoolByFloatComparison(A=$Engine.Owner.LODDistance, B=7500))
  ]
  rendererClasses [&NiagaraMeshRendererProperties]
  renderers [{rendererClass=&NiagaraMeshRendererProperties, rendererIndex=0}]
 }
}
summary {
 systemName=NS_Labrador_Anticipation
 emitters [
  {emitterName=Confetti, simTarget=GPUComputeSim, rendererClasses [&NiagaraSpriteRendererProperties]}
  {emitterName=Light, rendererClasses [&NiagaraLightRendererProperties]}
  {emitterName=Beam_Up_Build_Up, simTarget=GPUComputeSim, rendererClasses [&NiagaraSpriteRendererProperties]}
  {emitterName=Shimmer, rendererClasses [&NiagaraSpriteRendererProperties]}
  {emitterName=TinyFrags, rendererClasses [&NiagaraMeshRendererProperties]}
  {emitterName=PickUp_Poop, rendererClasses [&NiagaraMeshRendererProperties]}
 ]
 userVariables [
  User.LabradorMesh {
   type=NiagaraDataInterfaceSkeletalMesh
   defaultValue {
    struct=NiagaraExt_VariableValue_DataInterface
    value {dataInterface=&self:NiagaraDataInterfaceSkeletalMesh_0}
   }
  }
  User.SystemLife {type=NiagaraFloat, defaultValue=10, description=""}
 ]
}
