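# Niagara system for a cartoon dizzy effect, named for Labrador, combining orbiting star meshes, comic ring meshes and sparkle sprites.
# Stars burst once and orbit on two tilted axes until deactivation; fading rings spin at two tilts with opposite-signed material curves.
# GPU sparks flicker briefly on a sphere shell.
# All three cycle rainbow colors from engine time, with stars and rings varying cycle rate by spawn group.
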NiagaraSystem /Labrador/FX/NS_Labrador_Dizzy
dependencies {
 usedDataInterfaces [
  &NiagaraDataInterfaceMeshRendererInfo
  &NiagaraDataInterfaceCurve
  &NiagaraDataInterfaceColorCurve
  &NiagaraDataInterfaceArrayFloat
 ]
 usedDynamicInputs []
 usedModules [&SystemState]
}
emitters {
 Comic_Rings {
  emitterUpdateScript [
   &EmitterState(Life Cycle Mode=::NewEnumerator1, Loop Duration=0.75, UseLoopDelay=false)
   &SpawnRate(SpawnRate=6, Use Spawn Probability=false)
  ]
  particleSpawnScript [
   &FN_SetSpawnGroup(Number of Spawn Groups=2)
   &InitializeParticle(
    Color=<rgba 1 1 1 0.1>
    Lifetime=1.5
    Mesh Scale Mode=::NewEnumerator1
    Mesh Uniform Scale Max=0.39
    Mesh Uniform Scale Min=0.37
    UsePositionOffset=false
    Write Material Random=true
   )
   &InitialMeshOrientation(
    Rotation=&Lerp_Vector(A=<0.15 -0.05 0>, B=<0.15 0.05 0>, Alpha=&MakeFloatFromInt(Int=$Particles.SpawnGroup))
    Use Orientation Vector=true
    Use Rotation Vector=true
   )
  ]
  particleUpdateScript [
   &ParticleState()
   &ScaleColor(
    Scale Alpha=&FloatFromCurve(FloatCurve=DataInterface() {propertyValues {Curve {keys [<at 0 1 RCIM_Cubic>, {interpMode=RCIM_Cubic, leaveTangent=-1.8607, time=1, value=0}]}}})
    Scale RGB=&MakeVectorFromLinearColorRGB(Color=&ColorFromCurve(
      ColorCurve=DataInterface() {propertyValues {
        AlphaCurve {keys [<at 0 1 RCIM_Cubic>,<at 1 1 RCIM_Cubic>]}
        BlueCurve {keys [
          <at 0 0.6514 RCIM_Cubic>
          {arriveTangent=0.6972, interpMode=RCIM_Cubic, leaveTangent=0.6972, time=0.25, value=0.9216}
          {arriveTangent=-0.8685, interpMode=RCIM_Cubic, leaveTangent=-0.8685, time=0.5, value=1}
          {arriveTangent=-0.6983, interpMode=RCIM_Cubic, leaveTangent=-0.6983, time=0.75, value=0}
          <at 0.9992 0.6514 RCIM_Cubic>
         ]}
        GreenCurve {keys [
          <at 0 0.0497 RCIM_Cubic>
          {arriveTangent=1.83, interpMode=RCIM_Cubic, leaveTangent=1.83, time=0.25, value=0.0762}
          {arriveTangent=1.2671, interpMode=RCIM_Cubic, leaveTangent=1.2671, time=0.5, value=0.9647}
          {arriveTangent=-1.8329, interpMode=RCIM_Cubic, leaveTangent=-1.8329, time=0.75, value=1}
          <at 0.9992 0.0497 RCIM_Cubic>
         ]}
        RedCurve {keys [
          <at 0 1 RCIM_Cubic>
          {arriveTangent=-2, interpMode=RCIM_Cubic, leaveTangent=-2, time=0.25, value=0.013}
          <at 0.5 0 RCIM_Cubic>
          {arriveTangent=2.0033, interpMode=RCIM_Cubic, leaveTangent=2.0033, time=0.75, value=1}
          <at 0.9992 1 RCIM_Cubic>
         ]}
       }}
      CurveIndex=&Frac_Float(Float=&Multiply_Float(
        A=$Engine.Time
        B=&Add_Float(
         A=2
         B=&SelectFloatFromArray(Direct Array Index=$Particles.SpawnGroup, Float Selection Array=DataInterface() {propertyValues {FloatData [0,0.1]}})
        )
       ))
     ))
    ScaleA=true
    ScaleRGB=true
   )
   /SetVariables(Particles.BlackColor:LinearColor=&MultiplyLinearColor(Linear Color=$Particles.Color, Scale Factor=&MakeVector4_FromFloat(Float=1.5))) {isSetParameters=true}
   &ScaleMeshSize(Scale Factor=&VectorFromFloat(Value=&FloatFromCurve(FloatCurve=DataInterface() {propertyValues {Curve {keys [{arriveTangent=2.9993, interpMode=RCIM_Cubic, time=0, value=0}, <at 1 1 RCIM_Cubic>]}}}))) {enabled=false}
   &DynamicMaterialParameters(
    Index 0 Param 1=&FloatFromCurve(
     FloatCurve=DataInterface() {propertyValues {Curve {keys [<at 0 1>,<at 1 0>]}}}
     Scale Curve=&SelectFloatFromArray(Direct Array Index=$Particles.SpawnGroup, Float Selection Array=DataInterface() {propertyValues {FloatData [-1,1]}})
    )
    Index 0 Param 2=0.5
    Index 0 Param 3=1
    Index 1 Param 1=&FloatFromCurve(
     FloatCurve=DataInterface() {propertyValues {Curve {keys [<at 0 1>,<at 1 0>]}}}
     Scale Curve=&SelectFloatFromArray(Direct Array Index=$Particles.SpawnGroup, Float Selection Array=DataInterface() {propertyValues {FloatData [-1,1]}})
    )
    Index 1 Param 2=0.5
    Module.Param0WriteEnabled=true
    Module.Param0WriteEnabled1=true
    Module.Param1WriteEnabled=true
    Module.Param1WriteEnabled1=true
    Module.Param2WriteEnabled=true
    Module.Param2WriteEnabled1=true
    Module.Param3WriteEnabled=true
    Module.Param3WriteEnabled1=true
    Write Parameter Index 1=true
   )
   &UpdateMeshOrientation(Rotation Coordinate Space=::NewEnumerator3, Rotation Rate=1.5, Rotation Vector=<0 0 1>)
  ]
  rendererClasses [&NiagaraMeshRendererProperties]
  renderers [
   {rendererClass=&NiagaraMeshRendererProperties, rendererIndex=0}
   {rendererClass=&NiagaraMeshRendererProperties, rendererIndex=1}
  ]
 }
 Stars {
  emitterUpdateScript [
   &EmitterState(Life Cycle Mode=::NewEnumerator1, Loop Behavior=::NewEnumerator1, Loop Duration Mode=::NewEnumerator1, UseLoopDelay=false)
   &SpawnBurst_Instantaneous(Spawn Count=6, Use Spawn Probability=true)
  ]
  particleSpawnScript [
   &FN_SetSpawnGroup(Number of Spawn Groups=6)
   &InitializeParticle(
    Lifetime=0
    Mesh Scale Mode=::NewEnumerator1
    Mesh Uniform Scale Max=0.8
    Mesh Uniform Scale Min=0.6
    UsePositionOffset=false
    Write Material Random=true
   )
   &InitialMeshOrientation(Orientation Axis=<0 1 0>, Rotation Coordinate Space=::NewEnumerator0, Use Orientation Vector=true, Use Rotation Vector=true)
  ]
  particleUpdateScript [
   &ParticleState(Kill Particles When Lifetime Has Elapsed=false, Let Infinitely Lived Particles Die When Emitter Deactivates=true)
   &RotateAroundPoint(
    Derive the Rotation Axis=false
    Radius:NiagaraFloat=&RandomRangeFloat(Minimum=35, Maximum=45)
    Rotation Axis:Vector3f=&MakeCustomVectorFromBool(
     Bool=&SetBoolByIntComparison(A=$Particles.SpawnGroup, B=3, Comparison Type=::NewEnumerator2)
     True Vector=<-0.1 -0.325 0.25>
     False Vector=<0.1 -0.325 0.25>
    )
    Rotation Center:NiagaraPosition=/Niagara/DynamicInputs/Helpers/SimulationPosition()
    Rotation Phase:NiagaraFloat=&Multiply_Float(A=0.333, B=&MakeFloatFromInt(Int=$Particles.SpawnGroup))
    Rotation Rate=0.5
    X Axis:Vector3f=<0 1 0>
   ) {module=RotateAroundPoint001}
   &UpdateMeshOrientation(Rotation Coordinate Space=::NewEnumerator3, Rotation Rate=0.8, Rotation Vector=<0 0 1>) {enabled=false}
   &UpdateMeshOrientation(
    Rotation Coordinate Space=::NewEnumerator3
    Rotation Rate=&RandomRangeFloat(Minimum=0.1, Maximum=0.3)
    Rotation Vector=<0 1 0>
   ) {enabled=false, module=UpdateMeshOrientation001}
   &UpdateMeshOrientation(Rotation Rate=0.2, Rotation Vector=<0.5 0 0>) {enabled=false, module=UpdateMeshOrientation002}
   &ScaleMeshSize(Scale Factor=&VectorFromFloat(Value=&FloatFromCurve(FloatCurve=DataInterface() {propertyValues {Curve {keys [
          {arriveTangent=3.5393, interpMode=RCIM_Cubic, leaveTangent=3.5393, tangentMode=RCTM_User, time=0, value=0}
          {arriveTangent=20.3169, interpMode=RCIM_Cubic, leaveTangent=20.3169, tangentMode=RCTM_User, time=0.1, value=1}
          {interpMode=RCIM_Cubic, leaveTangent=-1.1111, time=0.2, value=1}
         ]}}})))
   &ScaleColor(
    Scale RGB=&MakeVectorFromLinearColorRGB(Color=&ColorFromCurve(
      ColorCurve=DataInterface() {propertyValues {
        AlphaCurve {keys [<at 0 1 RCIM_Cubic>,<at 1 1 RCIM_Cubic>]}
        BlueCurve {keys [
          <at 0 0.6514 RCIM_Cubic>
          {arriveTangent=0.6972, interpMode=RCIM_Cubic, leaveTangent=0.6972, time=0.25, value=0.9216}
          {arriveTangent=-0.8685, interpMode=RCIM_Cubic, leaveTangent=-0.8685, time=0.5, value=1}
          {arriveTangent=-0.6983, interpMode=RCIM_Cubic, leaveTangent=-0.6983, time=0.75, value=0}
          <at 0.9992 0.6514 RCIM_Cubic>
         ]}
        GreenCurve {keys [
          <at 0 0.0497 RCIM_Cubic>
          {arriveTangent=1.83, interpMode=RCIM_Cubic, leaveTangent=1.83, time=0.25, value=0.0762}
          {arriveTangent=1.2671, interpMode=RCIM_Cubic, leaveTangent=1.2671, time=0.5, value=0.9647}
          {arriveTangent=-1.8329, interpMode=RCIM_Cubic, leaveTangent=-1.8329, time=0.75, value=1}
          <at 0.9992 0.0497 RCIM_Cubic>
         ]}
        RedCurve {keys [
          <at 0 1 RCIM_Cubic>
          {arriveTangent=-2, interpMode=RCIM_Cubic, leaveTangent=-2, time=0.25, value=0.013}
          <at 0.5 0 RCIM_Cubic>
          {arriveTangent=2.0033, interpMode=RCIM_Cubic, leaveTangent=2.0033, time=0.75, value=1}
          <at 0.9992 1 RCIM_Cubic>
         ]}
       }}
      CurveIndex=&Frac_Float(Float=&Multiply_Float(
        A=$Engine.Time
        B=&Add_Float(
         A=2
         B=&SelectFloatFromArray(
          Direct Array Index=$Particles.SpawnGroup
          Float Selection Array=DataInterface() {propertyValues {FloatData [0,0.1,0.15,0.175,0.2,0.21,0.215]}}
         )
        )
       ))
     ))
    ScaleA=true
    ScaleRGB=true
   )
  ]
  rendererClasses [&NiagaraMeshRendererProperties]
  renderers [{rendererClass=&NiagaraMeshRendererProperties, rendererIndex=0}]
 }
 sparks {
  simTarget=GPUComputeSim
  emitterUpdateScript [
   &EmitterState()
   &SpawnRate(SpawnRate=&RandomRangeFloat(Minimum=15, Maximum=25), Use Spawn Probability=false)
  ]
  particleSpawnScript [
   &InitializeParticle(
    Lifetime Max=0.3
    Lifetime Min=0.1
    Lifetime Mode=::NewEnumerator1
    Sprite Rotation Mode=::NewEnumerator2
    Sprite Size Mode=::NewEnumerator1
    Uniform Sprite Size Max=15
    Uniform Sprite Size Min=7
    UsePositionOffset=false
    Write Material Random=true
   )
   &SphereLocation(
    MaskSpawnGroup=false
    Override Seed=false
    Sphere Radius=45
    Surface Only=true
    Surface Only Band Thickness=5
    UseNonUniformScale=false
    UseOffset=false
    UseSphereOrientation=false
   )
  ]
  particleUpdateScript [
   &ParticleState()
   &SpriteRotationRate(Rotation Rate=5000)
   &ScaleColor(
    Scale RGB=&VectorFromFloat(Value=&FloatFromCurve(
      FloatCurve=DataInterface() {propertyValues {Curve {keys [
          {arriveTangent=10.9401, interpMode=RCIM_Cubic, leaveTangent=10.9401, tangentMode=RCTM_User, time=0, value=0}
          <at 0.15 1 RCIM_Cubic>
          {arriveTangent=-1.0705, interpMode=RCIM_Cubic, leaveTangent=-1.0705, tangentMode=RCTM_User, time=1, value=0}
         ]}}}
      Scale Curve=100
     ))
    ScaleA=true
    ScaleRGB=true
   )
   &ScaleColor(
    Scale RGB=&MakeVectorFromLinearColorRGB(Color=&ColorFromCurve(
      ColorCurve=DataInterface() {propertyValues {
        AlphaCurve {keys [<at 0 1 RCIM_Cubic>,<at 1 1 RCIM_Cubic>]}
        BlueCurve {keys [
          <at 0 0.6514 RCIM_Cubic>
          {arriveTangent=0.6972, interpMode=RCIM_Cubic, leaveTangent=0.6972, time=0.25, value=0.9216}
          {arriveTangent=-0.8685, interpMode=RCIM_Cubic, leaveTangent=-0.8685, time=0.5, value=1}
          {arriveTangent=-0.6983, interpMode=RCIM_Cubic, leaveTangent=-0.6983, time=0.75, value=0}
          <at 0.9992 0.6514 RCIM_Cubic>
         ]}
        GreenCurve {keys [
          <at 0 0.0497 RCIM_Cubic>
          {arriveTangent=1.83, interpMode=RCIM_Cubic, leaveTangent=1.83, time=0.25, value=0.0762}
          {arriveTangent=1.2671, interpMode=RCIM_Cubic, leaveTangent=1.2671, time=0.5, value=0.9647}
          {arriveTangent=-1.8329, interpMode=RCIM_Cubic, leaveTangent=-1.8329, time=0.75, value=1}
          <at 0.9992 0.0497 RCIM_Cubic>
         ]}
        RedCurve {keys [
          <at 0 1 RCIM_Cubic>
          {arriveTangent=-2, interpMode=RCIM_Cubic, leaveTangent=-2, time=0.25, value=0.013}
          <at 0.5 0 RCIM_Cubic>
          {arriveTangent=2.0033, interpMode=RCIM_Cubic, leaveTangent=2.0033, time=0.75, value=1}
          <at 0.9992 1 RCIM_Cubic>
         ]}
       }}
      CurveIndex=&Frac_Float(Float=&Multiply_Float(A=$Engine.Time, B=2))
     ))
    ScaleA=true
    ScaleRGB=true
   ) {module=ScaleColor001}
   &ScaleSpriteSize(Scale Factor=&Vector2DFromFloat(Value=&FloatFromCurve(FloatCurve=DataInterface() {propertyValues {Curve {keys [
          <at 0 0 RCIM_Cubic>
          {arriveTangent=27.9899, interpMode=RCIM_Cubic, leaveTangent=27.9899, tangentMode=RCTM_User, time=0.1, value=1}
          <at 0.2 1 RCIM_Cubic>
         ]}}})))
  ]
  rendererClasses [&NiagaraSpriteRendererProperties]
  renderers [{rendererClass=&NiagaraSpriteRendererProperties}]
 }
}
summary {
 systemName=NS_Labrador_Dizzy
 emitters [
  {emitterName=Stars, rendererClasses [&NiagaraMeshRendererProperties]}
  {emitterName=sparks, simTarget=GPUComputeSim, rendererClasses [&NiagaraSpriteRendererProperties]}
  {emitterName=Comic_Rings, rendererClasses [&NiagaraMeshRendererProperties]}
 ]
 userVariables []
}
