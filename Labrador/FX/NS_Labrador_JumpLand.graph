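# NS_Labrador_JumpLand is a Niagara system for Labrador, named as a jump landing effect.
# It combines two emitters, NE_RadialSmoke_LW and NE_Confetti_LW, named for a radial smoke burst and confetti, with an LW suffix that suggests lightweight emitters.
# SystemState is the only listed module, and no user variables are exposed.
# Neither emitter lists renderers here, so their visual setup is not shown.
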NiagaraSystem /Labrador/FX/NS_Labrador_JumpLand
dependencies {usedDataInterfaces [], usedDynamicInputs [], usedModules [&SystemState]}
emitters {NE_Confetti_LW {rendererClasses [], renderers []}, NE_RadialSmoke_LW {rendererClasses [], renderers []}}
summary {
 systemName=NS_Labrador_JumpLand
 emitters [{emitterName=NE_RadialSmoke_LW, rendererClasses []}, {emitterName=NE_Confetti_LW, rendererClasses []}]
 userVariables []
}
