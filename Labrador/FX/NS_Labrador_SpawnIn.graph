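# Niagara spawn-in effect for Labrador: a glowing mesh-and-sprite element drops from high above to a system offset, with rainbow colors settling toward orange.
# A shared DropTime variable drives the fall's position, scale and color curves, and times the BaseFlare burst with its light.
# DropTime also gates two spark emitters: curl-noise GPU sparks and stretched sparks, which use spawn groups to separate falling streaks from upward spray.
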NiagaraSystem /Labrador/FX/NS_Labrador_SpawnIn
dependencies {
 usedDataInterfaces [&NiagaraDataInterfaceColorCurve, &NiagaraDataInterfaceCurve]
 usedDynamicInputs []
 usedModules [
  &self:SystemSpawnScript.SystemScriptSource.SystemScriptGraph.NiagaraNodeAssignment_1.SetVariables_#1
  &self:SystemSpawnScript.SystemScriptSource.SystemScriptGraph.NiagaraNodeAssignment_2.SetVariables_#2
  &SystemState
  &self:SystemSpawnScript.SystemScriptSource.SystemScriptGraph.NiagaraNodeAssignment_0.SetVariables_#3
 ]
}
emitters {
 BaseFlare {
  emitterUpdateScript [
   &EmitterState(Life Cycle Mode=::NewEnumerator1, Loop Behavior=::NewEnumerator1, Loop Duration=1, UseLoopDelay=false)
   &SpawnBurst_Instantaneous(Spawn Count=1, Spawn Time=$System.DropTime, Use Spawn Probability=false)
  ]
  particleSpawnScript [
   &InitializeParticle(
    Color Mode=::NewEnumerator0
    Lifetime=1.5
    Position Offset=$System.SystemOffset
    Sprite Size Mode=::NewEnumerator3
    Uniform Sprite Size=1000
    UsePositionOffset=true
    Write Material Random=true
   )
   &CameraOffset(Camera Offset Amount=150)
  ]
  particleUpdateScript [
   &ParticleState()
   &ScaleColor(
    Scale Alpha=&FloatFromCurve(FloatCurve=DataInterface() {propertyValues {Curve {keys [
         {arriveTangent=21.0514, interpMode=RCIM_Cubic, leaveTangent=21.0514, leaveTangentWeight=0.0872, tangentMode=RCTM_User, time=0, value=0}
         {arriveTangentWeight=0.0534, interpMode=RCIM_Cubic, leaveTangentWeight=0.3735, tangentMode=RCTM_User, time=0.1, value=1}
         {arriveTangentWeight=0.3488, interpMode=RCIM_Cubic, leaveTangentWeight=0.0997, tangentMode=RCTM_User, time=0.5, value=1}
         {arriveTangent=-0.1275, arriveTangentWeight=0.3, interpMode=RCIM_Cubic, leaveTangent=-0.1275, tangentMode=RCTM_User, time=1, value=0}
        ]}}})
    Scale RGB=&MakeVectorFromLinearColorRGB(Color=&Multiply_LinearColorByFloat(
      Linear Color=&LerpLinearColors(
       StartColor=&ColorFromCurve(
        ColorCurve=DataInterface() {propertyValues {
          AlphaCurve {keys [<at 0 1 RCIM_Cubic>,<at 1 1 RCIM_Cubic>]}
          BlueCurve {keys [
            <at 0 0.6514 RCIM_Cubic>
            {arriveTangent=0.6972, interpMode=RCIM_Cubic, leaveTangent=0.6972, time=0.25, value=0.9216}
            {arriveTangent=-1.8432, interpMode=RCIM_Cubic, leaveTangent=-1.8432, time=0.5, value=1}
            {arriveTangent=-0.6983, interpMode=RCIM_Cubic, leaveTangent=-0.6983, time=0.75, value=0}
            <at 0.9992 0.6514 RCIM_Cubic>
           ]}
          GreenCurve {keys [
            <at 0 0.0497 RCIM_Cubic>
            {arriveTangent=1.83, interpMode=RCIM_Cubic, leaveTangent=1.83, time=0.25, value=0.0762}
            {arriveTangent=1.2476, interpMode=RCIM_Cubic, leaveTangent=1.2476, time=0.5, value=0.9647}
            {arriveTangent=-1.8329, interpMode=RCIM_Cubic, leaveTangent=-1.8329, time=0.75, value=0.7}
            <at 0.9992 0.0497 RCIM_Cubic>
           ]}
          RedCurve {keys [
            <at 0 1 RCIM_Cubic>
            {arriveTangent=-2, interpMode=RCIM_Cubic, leaveTangent=-2, time=0.25, value=0.013}
            {arriveTangent=1.374, interpMode=RCIM_Cubic, leaveTangent=1.374, time=0.5, value=0}
            {arriveTangent=2.0033, interpMode=RCIM_Cubic, leaveTangent=2.0033, time=0.75, value=0.7}
            <at 0.9992 1 RCIM_Cubic>
           ]}
         }}
        CurveIndex=&Frac_Float(Float=&Multiply_Float(A=$Engine.Time, B=2))
       )
       EndColor=<rgba 1 0.302 0 1>
       LerpFactor=&FloatFromCurve(FloatCurve=DataInterface() {propertyValues {Curve {keys [<at 0.5 0>,<at 0.9 1>]}}})
      )
      Float=1
     ))
    ScaleA=true
    ScaleRGB=true
   )
   &ScaleSpriteSize(Scale Factor=&Vector2DFromFloat(Value=&FloatFromCurve(FloatCurve=DataInterface() {propertyValues {Curve {keys [
          {arriveTangent=12.078, interpMode=RCIM_Cubic, leaveTangent=12.078, tangentMode=RCTM_User, time=0, value=0}
          {arriveTangent=-3.9514, interpMode=RCIM_Cubic, leaveTangent=-3.9514, tangentMode=RCTM_User, time=0.1, value=1.1}
          <at 0.4 1 RCIM_Cubic>
         ]}}})))
   &FN_Light_Attributes(
    Color=$Particles.Color
    Create Color Attribute=true
    Create Exponent Attribute=true
    Create Radius Attribute=true
    Exponent=1.5
    Radius=&Multiply_Float(A=&MakeFloatFromVector2d(Vector 2D=$Particles.SpriteSize), B=0.3)
    Scale Color=&Lerp_Float(
     A=0
     B=30
     Alpha=&FloatFromCurve(FloatCurve=DataInterface() {propertyValues {Curve {keys [
          {arriveTangent=4.1807, interpMode=RCIM_Cubic, leaveTangent=4.1807, tangentMode=RCTM_User, time=0, value=0}
          {interpMode=RCIM_Cubic, tangentMode=RCTM_User, time=0.25, value=1}
          {arriveTangent=-1.5114, interpMode=RCIM_Cubic, leaveTangent=-1.5114, tangentMode=RCTM_User, time=1, value=0}
         ]}}})
    )
   )
   &FN_Sprite_Attributes(
    Color=&MultiplyLinearColor(Linear Color=$Particles.Color, Scale Factor=<1 1 1 0.5>)
    Create Color Attribute=true
    Create SpriteSize Attribute=true
    SpriteSize=&Multiply_Vector2D_ByFloat(Float=0.7, Vector2D=$Particles.SpriteSize)
    Write to Module Namespace=true
   )
  ]
  rendererClasses [&NiagaraSpriteRendererProperties, &NiagaraLightRendererProperties]
  renderers [
   {rendererClass=&NiagaraSpriteRendererProperties}
   {rendererClass=&NiagaraSpriteRendererProperties, rendererIndex=1}
   {rendererClass=&NiagaraLightRendererProperties, rendererIndex=2}
  ]
 }
 GPU_Sparks {
  simTarget=GPUComputeSim
  emitterUpdateScript [
   &EmitterState(Life Cycle Mode=::NewEnumerator1, Loop Behavior=::NewEnumerator1, Loop Duration=0.8, UseLoopDelay=false)
   &SpawnRate(
    SpawnRate=&MakeCustomFloatFromBool(Bool=&SetBoolByFloatComparison(A=$Engine.System.Age, B=$System.DropTime), True Float=400, False Float=0)
    Use Spawn Probability=false
   )
  ]
  particleSpawnScript [
   &InitializeParticle(
    Color=&ColorFromCurve(
     ColorCurve=DataInterface() {propertyValues {
       AlphaCurve {keys [<at 0 1 RCIM_Cubic>,<at 1 1 RCIM_Cubic>]}
       BlueCurve {keys [
         <at 0 0.6514 RCIM_Cubic>
         {arriveTangent=0.6972, interpMode=RCIM_Cubic, leaveTangent=0.6972, time=0.25, value=0.9216}
         {arriveTangent=-0.8685, interpMode=RCIM_Cubic, leaveTangent=-0.8685, time=0.5, value=1}
         {arriveTangent=-0.6983, interpMode=RCIM_Cubic, leaveTangent=-0.6983, time=0.75, value=0}
         <at 0.9992 0.6514 RCIM_Cubic>
        ]}
       GreenCurve {keys [
         <at 0 0.0497 RCIM_Cubic>
         {arriveTangent=1.83, interpMode=RCIM_Cubic, leaveTangent=1.83, time=0.25, value=0.0762}
         {arriveTangent=1.2671, interpMode=RCIM_Cubic, leaveTangent=1.2671, time=0.5, value=0.9647}
         {arriveTangent=-1.8329, interpMode=RCIM_Cubic, leaveTangent=-1.8329, time=0.75, value=1}
         <at 0.9992 0.0497 RCIM_Cubic>
        ]}
       RedCurve {keys [
         <at 0 1 RCIM_Cubic>
         {arriveTangent=-2, interpMode=RCIM_Cubic, leaveTangent=-2, time=0.25, value=0.013}
         <at 0.5 0 RCIM_Cubic>
         {arriveTangent=2.0033, interpMode=RCIM_Cubic, leaveTangent=2.0033, time=0.75, value=1}
         <at 0.9992 1 RCIM_Cubic>
        ]}
      }}
     CurveIndex=&RandomRangeFloat(Minimum=0, Maximum=1)
    )
    Lifetime Max=1.5
    Lifetime Min=0.3
    Lifetime Mode=::NewEnumerator1
    Position Offset=$System.SystemOffset
    Sprite Rotation Mode=::NewEnumerator2
    Sprite Size Mode=::NewEnumerator1
    Uniform Sprite Size Max=15
    Uniform Sprite Size Min=7
    UsePositionOffset=true
    Write Material Random=true
   )
   &SphereLocation(
    MaskSpawnGroup=false
    Override Seed=false
    Sphere Radius=75
    Surface Only=true
    Surface Only Band Thickness=5
    UseNonUniformScale=false
    UseOffset=false
    UseSphereOrientation=false
   )
   &AddVelocityFromPoint(Module.Use Falloff Distance=false)
   &AddVelocity(Coordinate Space=::World, Velocity=<0 0 120>)
  ]
  particleUpdateScript [
   &GravityForce(Gravity=<0 0 -250>)
   &CurlNoiseForce(Noise Frequency=10, Noise Strength=670, Pan Noise=true, Pan Noise Field=<0 0 10>)
   &SolveForcesAndVelocity(Clamp Velocity=false, Limit Acceleration=false)
   &ParticleState()
   &ScaleColor(
    Scale RGB=&VectorFromFloat(Value=&FloatFromCurve(
      FloatCurve=DataInterface() {propertyValues {Curve {keys [
          {arriveTangent=10.9401, interpMode=RCIM_Cubic, leaveTangent=10.9401, tangentMode=RCTM_User, time=0, value=0}
          <at 0.15 1 RCIM_Cubic>
          {arriveTangent=-1.0705, interpMode=RCIM_Cubic, leaveTangent=-1.0705, tangentMode=RCTM_User, time=1, value=0}
         ]}}}
      Scale Curve=&RandomRangeFloat(Minimum=1, Maximum=100, Evaluation Type=::NewEnumerator1)
     ))
    ScaleA=true
    ScaleRGB=true
   )
   &ScaleColor(
    Scale RGB=&MakeVectorFromLinearColorRGB(Color=&ColorFromCurve(
      ColorCurve=DataInterface() {propertyValues {
        AlphaCurve {keys [<at 0 1 RCIM_Cubic>,<at 1 1 RCIM_Cubic>]}
        BlueCurve {keys [
          <at 0 0.6514 RCIM_Cubic>
          {arriveTangent=0.6972, interpMode=RCIM_Cubic, leaveTangent=0.6972, time=0.25, value=0.9216}
          {arriveTangent=-0.8685, interpMode=RCIM_Cubic, leaveTangent=-0.8685, time=0.5, value=1}
          {arriveTangent=-0.6983, interpMode=RCIM_Cubic, leaveTangent=-0.6983, time=0.75, value=0}
          <at 0.9992 0.6514 RCIM_Cubic>
         ]}
        GreenCurve {keys [
          <at 0 0.0497 RCIM_Cubic>
          {arriveTangent=1.83, interpMode=RCIM_Cubic, leaveTangent=1.83, time=0.25, value=0.0762}
          {arriveTangent=1.2671, interpMode=RCIM_Cubic, leaveTangent=1.2671, time=0.5, value=0.9647}
          {arriveTangent=-1.8329, interpMode=RCIM_Cubic, leaveTangent=-1.8329, time=0.75, value=1}
          <at 0.9992 0.0497 RCIM_Cubic>
         ]}
        RedCurve {keys [
          <at 0 1 RCIM_Cubic>
          {arriveTangent=-2, interpMode=RCIM_Cubic, leaveTangent=-2, time=0.25, value=0.013}
          <at 0.5 0 RCIM_Cubic>
          {arriveTangent=2.0033, interpMode=RCIM_Cubic, leaveTangent=2.0033, time=0.75, value=1}
          <at 0.9992 1 RCIM_Cubic>
         ]}
       }}
      CurveIndex=&Frac_Float(Float=&Multiply_Float(A=$Engine.Time, B=2))
     ))
    ScaleA=true
    ScaleRGB=true
   ) {enabled=false, module=ScaleColor001}
   &ScaleSpriteSize(Scale Factor=&Vector2DFromFloat(Value=&FloatFromCurve(FloatCurve=DataInterface() {propertyValues {Curve {keys [
          <at 0 0 RCIM_Cubic>
          {arriveTangent=27.9899, interpMode=RCIM_Cubic, leaveTangent=27.9899, tangentMode=RCTM_User, time=0.1, value=1}
          <at 0.2 1 RCIM_Cubic>
         ]}}})))
  ]
  rendererClasses [&NiagaraSpriteRendererProperties]
  renderers [{rendererClass=&NiagaraSpriteRendererProperties}]
 }
 PickUp_Poop {
  emitterUpdateScript [
   &EmitterState(Life Cycle Mode=::NewEnumerator1, Loop Behavior=::NewEnumerator1, Loop Duration=2, UseLoopDelay=false)
   &SpawnBurst_Instantaneous(Spawn Count=1, Use Spawn Probability=false)
   &SpawnBurst_Instantaneous(Spawn Count=1, Spawn Group=1, Use Spawn Probability=false) {module=SpawnBurst_Instantaneous001}
  ]
  particleSpawnScript [
   &InitializeParticle(
    Color=<rgba 2 2 2 1>
    Lifetime=&MergeHelper_FloatSwitch(0=1, 1=1, 2=?, 3=?, Index=&SpawnGroup(), Num=::NewEnumerator0)
    Mesh Scale=<8 1 0.5>
    Mesh Scale Mode=::NewEnumerator4
    Sprite Size Mode=::NewEnumerator3
    Uniform Sprite Size=500
    UsePositionOffset=false
    Write Material Random=true
   )
   /SetVariables(Particles.VisibilityTag=&SpawnGroup()) {isSetParameters=true}
   &DynamicMaterialParameters(
    Index 0 Param 1=&RandomRangeFloat(Minimum=0, Maximum=5)
    Index 0 Param 2=5
    Index 0 Param 3=1
    Index 0 Param 4=1
    Module.Param0WriteEnabled=true
    Module.Param1WriteEnabled=true
    Module.Param2WriteEnabled=true
    Module.Param3WriteEnabled=false
   )
   &CameraOffset(Camera Offset Amount=100)
  ]
  particleUpdateScript [
   &ParticleState()
   /SetVariables(Particles.Position=&Lerp_Vector(
     A=$System.SystemOffset
     B=<0 0 3000>
     Alpha=&FloatFromCurve(
      FloatCurve=DataInterface() {propertyValues {Curve {keys [<at 0 1>,<at 1 0>]}}}
      CurveIndex=&Divide_Float(A=$Engine.System.Age, B=$System.DropTime)
     )
    )) {isSetParameters=true}
   &ScaleColor(
    Scale Alpha=&FloatFromCurve(
     FloatCurve=DataInterface() {propertyValues {Curve {keys [
         <at 0 0 RCIM_Cubic>
         {interpMode=RCIM_Cubic, tangentMode=RCTM_User, time=0.15, value=1}
         {arriveTangent=-0.0576, interpMode=RCIM_Cubic, leaveTangent=-0.0576, tangentMode=RCTM_User, time=1, value=1}
         {interpMode=RCIM_Cubic, tangentMode=RCTM_User, time=1.5, value=0}
        ]}}}
     CurveIndex=&Divide_Float(A=$Engine.System.Age, B=$System.DropTime)
    )
    Scale RGB=&MakeVectorFromLinearColorRGB(Color=&Multiply_LinearColorByFloat(
      Linear Color=&LerpLinearColors(
       StartColor=&ColorFromCurve(
        ColorCurve=DataInterface() {propertyValues {
          AlphaCurve {keys [<at 0 1 RCIM_Cubic>,<at 1 1 RCIM_Cubic>]}
          BlueCurve {keys [
            <at 0 0.9771 RCIM_Cubic>
            {arriveTangent=0.0458, interpMode=RCIM_Cubic, leaveTangent=0.0458, time=0.25, value=5}
            {arriveTangent=-10, interpMode=RCIM_Cubic, leaveTangent=-10, time=0.5, value=1}
            {arriveTangent=-0.0459, interpMode=RCIM_Cubic, leaveTangent=-0.0459, time=0.75, value=0}
            <at 0.9992 0.9771 RCIM_Cubic>
           ]}
          GreenCurve {keys [
            <at 0 0.0746 RCIM_Cubic>
            {arriveTangent=1.7803, interpMode=RCIM_Cubic, leaveTangent=1.7803, time=0.25, value=0.4133}
            {arriveTangent=1.1733, interpMode=RCIM_Cubic, leaveTangent=1.1733, time=0.5, value=0.9647}
            {arriveTangent=-1.7832, interpMode=RCIM_Cubic, leaveTangent=-1.7832, time=0.75, value=1}
            <at 0.9992 0.0746 RCIM_Cubic>
           ]}
          RedCurve {keys [
            <at 0 1.5 RCIM_Cubic>
            {arriveTangent=-3, interpMode=RCIM_Cubic, leaveTangent=-3, time=0.25, value=0.0704}
            {arriveTangent=1.8591, interpMode=RCIM_Cubic, leaveTangent=1.8591, time=0.5, value=0}
            {arriveTangent=3.0049, interpMode=RCIM_Cubic, leaveTangent=3.0049, time=0.75, value=1}
            <at 0.9992 1.5 RCIM_Cubic>
           ]}
         }}
        CurveIndex=&Frac_Float(Float=&Divide_Float(A=$Engine.System.Age, B=$System.DropTime))
       )
       EndColor=<rgba 1 0.3008 0 1>
       LerpFactor=&FloatFromCurve(
        FloatCurve=DataInterface() {propertyValues {Curve {keys [<at 1 0>,<at 1.2 1>]}}}
        CurveIndex=&Divide_Float(A=$Engine.System.Age, B=$System.DropTime)
       )
      )
      Float=1
     ))
    ScaleA=true
    ScaleRGB=true
   )
   &ScaleMeshSize(Scale Factor=&VectorFromFloat(Value=&FloatFromCurve(
      FloatCurve=DataInterface() {propertyValues {Curve {keys [
          {arriveTangent=1.1546, interpMode=RCIM_Cubic, leaveTangent=1.1546, tangentMode=RCTM_User, time=0, value=0}
          {arriveTangent=-0.0291, interpMode=RCIM_Cubic, leaveTangent=-0.0291, tangentMode=RCTM_User, time=1, value=1}
          {arriveTangent=-1.3264, interpMode=RCIM_Cubic, leaveTangent=-1.3264, tangentMode=RCTM_User, time=3, value=0}
         ]}}}
      CurveIndex=&Divide_Float(A=$Engine.System.Age, B=$System.DropTime)
     )))
   &DynamicMaterialParameters(
    Index 2 Param 1=&Multiply_Float(A=&MakeFloatFromVector(Vector=$Particles.Scale, Channel=::NewEnumerator1), B=300)
    Index 3 Param 1=&Multiply_Float(A=&MakeFloatFromVector(Vector=$Particles.Scale, Channel=::NewEnumerator1), B=80)
    Module.Index2Float0Write=true
    Module.Index2Float1Write=true
    Module.Index2Float2Write=true
    Module.Index2Float3Write=true
    Module.Index3Float0Write=true
    Module.Index3Float1Write=true
    Module.Index3Float2Write=true
    Module.Index3Float3Write=true
    Write Parameter Index 0=::NewEnumerator0
    Write Parameter Index 2=::NewEnumerator1
    Write Parameter Index 3=::NewEnumerator1
   ) {module=DynamicMaterialParameters001}
   &FN_Mesh_Attributes(
    Color=&MultiplyLinearColor(Linear Color=$Particles.Color, Scale Factor=<1 1 1 0.1>)
    Create Color Attribute=true
    Write to Module Namespace=true
   )
   &FN_Sprite_Attributes(
    Color=&MultiplyLinearColor(Linear Color=$Particles.Color, Scale Factor=<1 1 1 0.5>)
    Create Color Attribute=true
    Create SpriteSize Attribute=true
    SpriteSize=&Multiply_Vector2D_ByFloat(Float=0.7, Vector2D=$Particles.SpriteSize)
    Write to Module Namespace=true
   )
   /SetVariables(
    Particles.CoronaDynMat1:Vector4f=&MakeVector4(X=0.5, Y=1, Z=1, W=1)
    Particles.CoronaDynMat2:Vector4f=&MakeVector4(X=6, Y=0, Z=1, W=1)
   ) {isSetParameters=true}
  ]
  rendererClasses [&NiagaraMeshRendererProperties, &NiagaraSpriteRendererProperties]
  renderers [
   {rendererClass=&NiagaraMeshRendererProperties, rendererIndex=0}
   {rendererClass=&NiagaraMeshRendererProperties, rendererIndex=1}
   {rendererClass=&NiagaraSpriteRendererProperties, rendererIndex=2}
   {rendererClass=&NiagaraSpriteRendererProperties, rendererIndex=3}
   {rendererClass=&NiagaraMeshRendererProperties, rendererIndex=4}
  ]
 }
 Stretch_Sparks {
  emitterUpdateScript [
   &EmitterState(Life Cycle Mode=::NewEnumerator1, Loop Behavior=::NewEnumerator1, Loop Duration=0.8, UseLoopDelay=false)
   &SpawnRate(
    SpawnGroup=1
    SpawnRate=&MakeCustomFloatFromBool(Bool=&SetBoolByFloatComparison(A=$Engine.System.Age, B=$System.DropTime), True Float=60, False Float=0)
    Use Spawn Probability=false
   )
   &SpawnBurst_Instantaneous(Spawn Count=6, Use Spawn Probability=false)
  ]
  particleSpawnScript [
   &InitializeParticle(
    Color=<rgba 4 4 4 1>
    Lifetime Max=0.6
    Lifetime Min=0.5
    Lifetime Mode=::NewEnumerator1
    Position Offset=$System.SystemOffset
    Sprite Size Max=<1.5 5>
    Sprite Size Min=<0.5 3>
    Sprite Size Mode=::NewEnumerator2
    UsePositionOffset=true
    Write Material Random=true
   )
   /SetVariables(Particles.Age:NiagaraFloat=&RandomRangeFloat(Minimum=0, Maximum=-0.15)) {isSetParameters=true}
   &SphereLocation(
    MaskSpawnGroup=false
    Override Seed=false
    Sphere Radius=75
    Surface Only=false
    UseNonUniformScale=false
    UseOffset=false
    UseSphereOrientation=false
   )
   &AddVelocity(Coordinate Space=::World, Velocity=&RandomRangeVector(Minimum=<0 0 750>, Maximum=<0 0 1500>))
  ]
  particleUpdateScript [
   &Drag(Drag=3, Use Linear Drag=true, Use Rotational Drag=true)
   &SolveForcesAndVelocity(Clamp Velocity=false, Limit Acceleration=false)
   &ParticleState()
   &ScaleSpriteSize(Scale Factor=&MaskVector2dBySpawnGroup(
     Masked Vector2D=&MakeVector2D(
      X=10
      Y=&FloatFromCurve(
       FloatCurve=DataInterface() {propertyValues {Curve {keys [<at 0.25 0>,<at 0.7 1>,<at 1 1>,<at 1.3 0>]}}}
       CurveIndex=&Divide_Float(A=$Engine.System.Age, B=$System.DropTime)
       Scale Curve=200
      )
     )
     Passthrough Vector2D=&MakeVector2D(
      X=5
      Y=&FloatFromCurve(FloatCurve=DataInterface() {propertyValues {Curve {keys [<at 0 1>,<at 1 0>]}}}, Scale Curve=10)
     )
     Spawn Group Mask=0
    ))
   &ScaleColor(
    Scale RGB=&MakeVectorFromLinearColorRGB(Color=&ColorFromCurve(
      ColorCurve=DataInterface() {propertyValues {
        AlphaCurve {keys [<at 0 1 RCIM_Cubic>,<at 1 1 RCIM_Cubic>]}
        BlueCurve {keys [
          <at 0 0.6514 RCIM_Cubic>
          {arriveTangent=0.6972, interpMode=RCIM_Cubic, leaveTangent=0.6972, time=0.25, value=0.9216}
          {arriveTangent=-0.8685, interpMode=RCIM_Cubic, leaveTangent=-0.8685, time=0.5, value=1}
          {arriveTangent=-0.6983, interpMode=RCIM_Cubic, leaveTangent=-0.6983, time=0.75, value=0}
          <at 0.9992 0.6514 RCIM_Cubic>
         ]}
        GreenCurve {keys [
          <at 0 0.0497 RCIM_Cubic>
          {arriveTangent=1.83, interpMode=RCIM_Cubic, leaveTangent=1.83, time=0.25, value=0.0762}
          {arriveTangent=1.2671, interpMode=RCIM_Cubic, leaveTangent=1.2671, time=0.5, value=0.9647}
          {arriveTangent=-1.8329, interpMode=RCIM_Cubic, leaveTangent=-1.8329, time=0.75, value=1}
          <at 0.9992 0.0497 RCIM_Cubic>
         ]}
        RedCurve {keys [
          <at 0 1 RCIM_Cubic>
          {arriveTangent=-2, interpMode=RCIM_Cubic, leaveTangent=-2, time=0.25, value=0.013}
          <at 0.5 0 RCIM_Cubic>
          {arriveTangent=2.0033, interpMode=RCIM_Cubic, leaveTangent=2.0033, time=0.75, value=1}
          <at 0.9992 1 RCIM_Cubic>
         ]}
       }}
      CurveIndex=&Frac_Float(Float=&Multiply_Float(A=$Engine.Time, B=2))
     ))
    ScaleA=true
    ScaleRGB=true
   )
   /SetVariables(Particles.Position=&MaskVectorBySpawnGroup(
     Masked Vector=&AddVectorToPosition(
      Position=&AddVectorToPosition(
       Position=$Engine.Owner.Position
       Vector=&Lerp_Vector(
        A=$System.SystemOffset
        B=&AddVectorToPosition(Position=<0 0 3000>, Vector=&RandomRangeVector(Minimum=<0 0 0>, Maximum=<1 1 1000>))
        Alpha=&FloatFromCurve(
         FloatCurve=DataInterface() {propertyValues {Curve {keys [<at 0 1>,<at 0.9 0>]}}}
         CurveIndex=&Divide_Float(A=$Particles.Age, B=$System.DropTime)
        )
       )
      )
      Vector=&Multiply_Vector(A=&RandomVector(Vector Scale=75), B=<1 1 0>)
     )
     Passthrough Vector=Linked() {linkedVariable {name=Particles.Position, type=NiagaraPosition}}
    )) {isSetParameters=true}
   &FN_Sprite_Attributes(
    Color=&MultiplyLinearColor(Linear Color=$Particles.Color, Scale Factor=<1 1 1 0.02>)
    Create Color Attribute=true
    Create SpriteSize Attribute=true
    SpriteSize=&Multiply_Vector2D(
     A=$Particles.SpriteSize
     B=&MaskVector2dBySpawnGroup(Masked Vector2D=<10 1.2>, Passthrough Vector2D=<3 1.2>, Spawn Group Mask=0)
    )
    Write to Module Namespace=true
   )
  ]
  rendererClasses [&NiagaraSpriteRendererProperties]
  renderers [
   {rendererClass=&NiagaraSpriteRendererProperties}
   {rendererClass=&NiagaraSpriteRendererProperties, rendererIndex=1}
  ]
 }
}
summary {
 systemName=NS_Labrador_SpawnIn
 emitters [
  {emitterName=BaseFlare, rendererClasses [&NiagaraSpriteRendererProperties, &NiagaraLightRendererProperties]}
  {emitterName=PickUp_Poop, rendererClasses [&NiagaraMeshRendererProperties, &NiagaraSpriteRendererProperties]}
  {emitterName=Stretch_Sparks, rendererClasses [&NiagaraSpriteRendererProperties]}
  {emitterName=GPU_Sparks, simTarget=GPUComputeSim, rendererClasses [&NiagaraSpriteRendererProperties]}
 ]
 userVariables []
}
