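# Niagara system for the Labrador teleport-out effect: a rainbow flash that shoots upward and fades.
# One-shot bursts form the corona, flare, light, beam mesh and a shrinking PickUp_Poop mesh launched skyward.
# Short spawn windows add rising stretched sparks and curl-noise GPU sparks.
# Most emitters cycle hue by indexing a color curve with engine time; GPU sparks pick random spawn hues, while the corona follows its own gradient.
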NiagaraSystem /Labrador/FX/NS_Labrador_TeleportOut
dependencies {
 usedDataInterfaces [&NiagaraDataInterfaceColorCurve, &NiagaraDataInterfaceCurve]
 usedDynamicInputs []
 usedModules [
  &SystemState
  &self:SystemSpawnScript.SystemScriptSource.SystemScriptGraph.NiagaraNodeAssignment_0.SetVariables_#1
 ]
}
emitters {
 Corona {
  emitterUpdateScript [
   &EmitterState()
   &SpawnBurst_Instantaneous(Spawn Count=1, Use Spawn Probability=false)
  ]
  particleSpawnScript [&InitializeParticle(Color=<rgba 2 2 2 1>, Sprite Size Mode=::NewEnumerator3, Uniform Sprite Size=300, UsePositionOffset=false, Write Material Random=true)]
  particleUpdateScript [
   &ParticleState()
   &ScaleColor(
    Scale Alpha=&FloatFromCurve(FloatCurve=DataInterface() {propertyValues {Curve {keys [
         {arriveTangent=25.6036, interpMode=RCIM_Cubic, leaveTangent=25.6036, tangentMode=RCTM_User, time=0, value=0}
         {arriveTangentWeight=0.0556, interpMode=RCIM_Cubic, leaveTangentWeight=0.3611, tangentMode=RCTM_User, time=0.1, value=1}
         {arriveTangentWeight=2.0573, interpMode=RCIM_Cubic, leaveTangentWeight=0.7913, tangentMode=RCTM_User, time=0.75, value=1}
         <at 1 0 RCIM_Cubic>
        ]}}})
    Scale RGB=&MakeVectorFromLinearColorRGB(Color=&ColorFromCurve(ColorCurve=DataInterface() {propertyValues {
        AlphaCurve {keys []}
        BlueCurve {keys [<at 0.25 0.6267 RCIM_Cubic>,<at 0.75 1 RCIM_Cubic>]}
        GreenCurve {keys [<at 0.25 0.6267 RCIM_Cubic>,<at 0.75 0.5219 RCIM_Cubic>]}
        RedCurve {keys [<at 0.25 1 RCIM_Cubic>,<at 0.75 0.2031 RCIM_Cubic>]}
       }}))
    ScaleA=true
    ScaleRGB=true
   )
   &DynamicMaterialParameters(
    Index 0 Param 1=&FloatFromCurve(FloatCurve=DataInterface() {propertyValues {Curve {keys [
         {arriveTangent=-4.8433, interpMode=RCIM_Cubic, leaveTangent=-4.8433, tangentMode=RCTM_User, time=0.5, value=1.5}
         {arriveTangentWeight=0.509, interpMode=RCIM_Cubic, tangentMode=RCTM_User, time=1, value=0.5}
        ]}}})
    Index 0 Param 2=1
    Index 0 Param 3=0.25
    Index 0 Param 4=1
    Module.Param0WriteEnabled=true
    Module.Param1WriteEnabled=true
    Module.Param2WriteEnabled=true
    Module.Param3WriteEnabled=true
   )
   &ScaleSpriteSize(Scale Factor=&Vector2DFromFloat(Value=&FloatFromCurve(FloatCurve=DataInterface() {propertyValues {Curve {keys [{arriveTangent=4.4751, interpMode=RCIM_Cubic, leaveTangent=4.4751, tangentMode=RCTM_User, time=0, value=0}, <at 1 2 RCIM_Cubic>]}}})))
  ]
  rendererClasses [&NiagaraSpriteRendererProperties]
  renderers [{rendererClass=&NiagaraSpriteRendererProperties}]
 }
 Light {
  emitterUpdateScript [
   &EmitterState()
   &SpawnBurst_Instantaneous(Spawn Count=1, Use Spawn Probability=false)
  ]
  particleSpawnScript [&InitializeParticle(UsePositionOffset=false, Write Material Random=true)]
  particleUpdateScript [
   &ParticleState()
   &ScaleColor(
    Scale RGB=&MakeVectorFromLinearColorRGB(Color=&ColorFromCurve(
      ColorCurve=DataInterface() {propertyValues {
        AlphaCurve {keys [<at 0 1 RCIM_Cubic>,<at 1 1 RCIM_Cubic>]}
        BlueCurve {keys [
          <at 0 0.6514 RCIM_Cubic>
          {arriveTangent=0.6972, interpMode=RCIM_Cubic, leaveTangent=0.6972, time=0.25, value=0.9216}
          {arriveTangent=-0.8685, interpMode=RCIM_Cubic, leaveTangent=-0.8685, time=0.5, value=1}
          {arriveTangent=-0.6983, interpMode=RCIM_Cubic, leaveTangent=-0.6983, time=0.75, value=0}
          <at 0.9992 0.6514 RCIM_Cubic>
         ]}
        GreenCurve {keys [
          <at 0 0.0497 RCIM_Cubic>
          {arriveTangent=1.83, interpMode=RCIM_Cubic, leaveTangent=1.83, time=0.25, value=0.0762}
          {arriveTangent=1.2671, interpMode=RCIM_Cubic, leaveTangent=1.2671, time=0.5, value=0.9647}
          {arriveTangent=-1.8329, interpMode=RCIM_Cubic, leaveTangent=-1.8329, time=0.75, value=1}
          <at 0.9992 0.0497 RCIM_Cubic>
         ]}
        RedCurve {keys [
          <at 0 1 RCIM_Cubic>
          {arriveTangent=-2, interpMode=RCIM_Cubic, leaveTangent=-2, time=0.25, value=0.013}
          <at 0.5 0 RCIM_Cubic>
          {arriveTangent=2.0033, interpMode=RCIM_Cubic, leaveTangent=2.0033, time=0.75, value=1}
          <at 0.9992 1 RCIM_Cubic>
         ]}
       }}
      CurveIndex=&Frac_Float(Float=&Multiply_Float(A=$Engine.Time, B=2))
     ))
    ScaleA=true
    ScaleRGB=true
   )
   &FN_Light_Attributes(
    Color=$Particles.Color
    Create Color Attribute=true
    Create Exponent Attribute=true
    Create Radius Attribute=true
    Exponent=3
    Radius=&Multiply_Float(
     A=150
     B=&FloatFromCurve(FloatCurve=DataInterface() {propertyValues {Curve {keys [<at 0.5 1>,<at 1 2>]}}})
    )
    Scale Color=150
   )
  ]
  rendererClasses [&NiagaraLightRendererProperties]
  renderers [{rendererClass=&NiagaraLightRendererProperties, rendererIndex=0}]
 }
 PickUp_Poop {
  emitterUpdateScript [
   &EmitterState(Life Cycle Mode=::NewEnumerator1, Loop Behavior=::NewEnumerator1, Loop Delay=0.05, Loop Duration=1, UseLoopDelay=true)
   &SpawnBurst_Instantaneous(Spawn Count=1, Use Spawn Probability=false)
  ]
  particleSpawnScript [
   &InitializeParticle(
    Color=<rgba 2 2 2 1>
    Lifetime=0.6
    Mesh Scale Mode=::NewEnumerator3
    Mesh Uniform Scale=200
    UsePositionOffset=false
    Write Material Random=true
   )
   &DynamicMaterialParameters(
    Index 0 Param 1=&RandomRangeFloat(Minimum=0, Maximum=5)
    Index 0 Param 2=5
    Index 0 Param 3=1
    Index 0 Param 4=1
    Module.Param0WriteEnabled=true
    Module.Param1WriteEnabled=true
    Module.Param2WriteEnabled=true
    Module.Param3WriteEnabled=false
   )
   &CameraOffset(Camera Offset Amount=100)
   &AddVelocity(Coordinate Space=::World, Velocity=<0 0 5000>)
  ]
  particleUpdateScript [
   &ParticleState()
   &AccelerationForce(Acceleration=<0 0 1600>)
   &SolveForcesAndVelocity(Clamp Velocity=false, Limit Acceleration=false)
   &ScaleColor(
    Scale Alpha=&FloatFromCurve(FloatCurve=DataInterface() {propertyValues {Curve {keys [{interpMode=RCIM_Cubic, tangentMode=RCTM_User, time=0, value=1}, {interpMode=RCIM_Cubic, tangentMode=RCTM_User, time=1, value=0}]}}})
    Scale RGB=&MakeVectorFromLinearColorRGB(Color=&ColorFromCurve(
      ColorCurve=DataInterface() {propertyValues {
        AlphaCurve {keys [<at 0 1 RCIM_Cubic>,<at 1 1 RCIM_Cubic>]}
        BlueCurve {keys [
          <at 0 0.6514 RCIM_Cubic>
          {arriveTangent=0.6972, interpMode=RCIM_Cubic, leaveTangent=0.6972, time=0.25, value=0.9216}
          {arriveTangent=-0.8685, interpMode=RCIM_Cubic, leaveTangent=-0.8685, time=0.5, value=1}
          {arriveTangent=-0.6983, interpMode=RCIM_Cubic, leaveTangent=-0.6983, time=0.75, value=0}
          <at 0.9992 0.6514 RCIM_Cubic>
         ]}
        GreenCurve {keys [
          <at 0 0.0497 RCIM_Cubic>
          {arriveTangent=1.83, interpMode=RCIM_Cubic, leaveTangent=1.83, time=0.25, value=0.0762}
          {arriveTangent=1.2671, interpMode=RCIM_Cubic, leaveTangent=1.2671, time=0.5, value=0.9647}
          {arriveTangent=-1.8329, interpMode=RCIM_Cubic, leaveTangent=-1.8329, time=0.75, value=1}
          <at 0.9992 0.0497 RCIM_Cubic>
         ]}
        RedCurve {keys [
          <at 0 1 RCIM_Cubic>
          {arriveTangent=-2, interpMode=RCIM_Cubic, leaveTangent=-2, time=0.25, value=0.013}
          <at 0.5 0 RCIM_Cubic>
          {arriveTangent=2.0033, interpMode=RCIM_Cubic, leaveTangent=2.0033, time=0.75, value=1}
          <at 0.9992 1 RCIM_Cubic>
         ]}
       }}
      CurveIndex=&Frac_Float(Float=&Multiply_Float(A=$Engine.Time, B=2))
     ))
    ScaleA=true
    ScaleRGB=true
   )
   &ScaleMeshSize(Scale Factor=&VectorFromFloat(Value=&FloatFromCurve(FloatCurve=DataInterface() {propertyValues {Curve {keys [
          {arriveTangent=-7.7187, interpMode=RCIM_Cubic, time=0, value=2}
          {arriveTangent=-3.6396, interpMode=RCIM_Cubic, leaveTangent=-3.6396, tangentMode=RCTM_User, time=1, value=0}
         ]}}})))
   &UpdateMeshOrientation(Rotation Rate=0.1, Rotation Vector=<1 0 0>)
  ]
  rendererClasses [&NiagaraMeshRendererProperties]
  renderers [{rendererClass=&NiagaraMeshRendererProperties, rendererIndex=0}]
 }
 Stretch_Sparks {
  emitterUpdateScript [
   &EmitterState(Life Cycle Mode=::NewEnumerator1, Loop Behavior=::NewEnumerator1, Loop Duration=0.8, UseLoopDelay=false)
   &SpawnRate(SpawnRate=60, Use Spawn Probability=false)
   &SpawnBurst_Instantaneous(Spawn Count=5, Use Spawn Probability=false)
  ]
  particleSpawnScript [
   &InitializeParticle(
    Color=<rgba 5 5 5 1>
    Lifetime Max=0.8
    Lifetime Min=0.1
    Lifetime Mode=::NewEnumerator1
    Sprite Size Max=<1.5 5>
    Sprite Size Min=<0.5 3>
    Sprite Size Mode=::NewEnumerator2
    UsePositionOffset=false
    Write Material Random=true
   )
   &SphereLocation(
    MaskSpawnGroup=false
    Override Seed=false
    Sphere Radius=75
    Surface Only=false
    UseNonUniformScale=false
    UseOffset=false
    UseSphereOrientation=false
   )
   &AddVelocity(Coordinate Space=::World, Velocity=&RandomRangeVector(Minimum=<0 0 80>, Maximum=<0 0 120>))
  ]
  particleUpdateScript [
   &AccelerationForce(Acceleration=<0 0 1600>)
   &SolveForcesAndVelocity(Clamp Velocity=false, Limit Acceleration=false)
   &ParticleState()
   &ScaleSpriteSize(Scale Factor=&MakeVector2D(
     X=&FloatFromCurve(FloatCurve=DataInterface() {propertyValues {Curve {keys [
          {arriveTangent=-11.4426, interpMode=RCIM_Cubic, leaveTangent=-11.4426, tangentMode=RCTM_User, time=0, value=10}
          {arriveTangent=-3.8142, interpMode=RCIM_Cubic, leaveTangent=-3.8142, tangentMode=RCTM_User, time=1, value=0}
         ]}}})
     Y=&FloatFromCurve(
      FloatCurve=DataInterface() {propertyValues {Curve {keys [<at 0 0 RCIM_Cubic>, {arriveTangent=183.081, interpMode=RCIM_Cubic, leaveTangent=183.081, tangentMode=RCTM_User, time=1, value=80}]}}}
      Scale Curve=5
     )
    ))
   &ScaleColor(
    Scale RGB=&MakeVectorFromLinearColorRGB(Color=&ColorFromCurve(
      ColorCurve=DataInterface() {propertyValues {
        AlphaCurve {keys [<at 0 1 RCIM_Cubic>,<at 1 1 RCIM_Cubic>]}
        BlueCurve {keys [
          <at 0 0.6514 RCIM_Cubic>
          {arriveTangent=0.6972, interpMode=RCIM_Cubic, leaveTangent=0.6972, time=0.25, value=0.9216}
          {arriveTangent=-0.8685, interpMode=RCIM_Cubic, leaveTangent=-0.8685, time=0.5, value=1}
          {arriveTangent=-0.6983, interpMode=RCIM_Cubic, leaveTangent=-0.6983, time=0.75, value=0}
          <at 0.9992 0.6514 RCIM_Cubic>
         ]}
        GreenCurve {keys [
          <at 0 0.0497 RCIM_Cubic>
          {arriveTangent=1.83, interpMode=RCIM_Cubic, leaveTangent=1.83, time=0.25, value=0.0762}
          {arriveTangent=1.2671, interpMode=RCIM_Cubic, leaveTangent=1.2671, time=0.5, value=0.9647}
          {arriveTangent=-1.8329, interpMode=RCIM_Cubic, leaveTangent=-1.8329, time=0.75, value=1}
          <at 0.9992 0.0497 RCIM_Cubic>
         ]}
        RedCurve {keys [
          <at 0 1 RCIM_Cubic>
          {arriveTangent=-2, interpMode=RCIM_Cubic, leaveTangent=-2, time=0.25, value=0.013}
          <at 0.5 0 RCIM_Cubic>
          {arriveTangent=2.0033, interpMode=RCIM_Cubic, leaveTangent=2.0033, time=0.75, value=1}
          <at 0.9992 1 RCIM_Cubic>
         ]}
       }}
      CurveIndex=&Frac_Float(Float=&Multiply_Float(A=$Engine.Time, B=2))
     ))
    ScaleA=true
    ScaleRGB=true
   )
  ]
  rendererClasses [&NiagaraSpriteRendererProperties]
  renderers [{rendererClass=&NiagaraSpriteRendererProperties}]
 }
 Sweeping_Flares {
  emitterUpdateScript [
   &EmitterState()
   &SpawnBurst_Instantaneous(Spawn Count=1, Use Spawn Probability=false)
  ]
  particleSpawnScript [
   &InitializeParticle(
    Color Mode=::NewEnumerator0
    Lifetime=1.5
    Sprite Size Mode=::NewEnumerator3
    Uniform Sprite Size=100
    UsePositionOffset=false
    Write Material Random=true
   )
   &CameraOffset(Camera Offset Amount=150)
  ]
  particleUpdateScript [
   &ParticleState()
   &ScaleColor(
    Scale Alpha=&FloatFromCurve(
     FloatCurve=DataInterface() {propertyValues {Curve {keys [
         {arriveTangent=21.0514, interpMode=RCIM_Cubic, leaveTangent=21.0514, leaveTangentWeight=0.0872, tangentMode=RCTM_User, time=0, value=0}
         {arriveTangentWeight=0.0534, interpMode=RCIM_Cubic, leaveTangentWeight=0.3735, tangentMode=RCTM_User, time=0.1, value=1}
         {arriveTangentWeight=0.3488, interpMode=RCIM_Cubic, leaveTangentWeight=0.0997, tangentMode=RCTM_User, time=0.5, value=1}
         {arriveTangentWeight=0.3, interpMode=RCIM_Cubic, tangentMode=RCTM_User, time=1, value=0}
        ]}}}
     Scale Curve=10
    )
    Scale RGB=&MakeVectorFromLinearColorRGB(Color=&ColorFromCurve(
      ColorCurve=DataInterface() {propertyValues {
        AlphaCurve {keys [<at 0 1 RCIM_Cubic>,<at 1 1 RCIM_Cubic>]}
        BlueCurve {keys [
          <at 0 0.6514 RCIM_Cubic>
          {arriveTangent=0.6972, interpMode=RCIM_Cubic, leaveTangent=0.6972, time=0.25, value=0.9216}
          {arriveTangent=-0.8685, interpMode=RCIM_Cubic, leaveTangent=-0.8685, time=0.5, value=1}
          {arriveTangent=-0.6983, interpMode=RCIM_Cubic, leaveTangent=-0.6983, time=0.75, value=0}
          <at 0.9992 0.6514 RCIM_Cubic>
         ]}
        GreenCurve {keys [
          <at 0 0.0497 RCIM_Cubic>
          {arriveTangent=1.83, interpMode=RCIM_Cubic, leaveTangent=1.83, time=0.25, value=0.0762}
          {arriveTangent=1.2671, interpMode=RCIM_Cubic, leaveTangent=1.2671, time=0.5, value=0.9647}
          {arriveTangent=-1.8329, interpMode=RCIM_Cubic, leaveTangent=-1.8329, time=0.75, value=1}
          <at 0.9992 0.0497 RCIM_Cubic>
         ]}
        RedCurve {keys [
          <at 0 1 RCIM_Cubic>
          {arriveTangent=-2, interpMode=RCIM_Cubic, leaveTangent=-2, time=0.25, value=0.013}
          <at 0.5 0 RCIM_Cubic>
          {arriveTangent=2.0033, interpMode=RCIM_Cubic, leaveTangent=2.0033, time=0.75, value=1}
          <at 0.9992 1 RCIM_Cubic>
         ]}
       }}
      CurveIndex=&Frac_Float(Float=&Multiply_Float(A=$Engine.Time, B=2))
     ))
    ScaleA=true
    ScaleRGB=true
   )
   &ScaleSpriteSize(Scale Factor=&Vector2DFromFloat(Value=&FloatFromCurve(FloatCurve=DataInterface() {propertyValues {Curve {keys [{arriveTangent=9.2569, interpMode=RCIM_Cubic, leaveTangent=9.2569, tangentMode=RCTM_User, time=0, value=5}, <at 0.5 10 RCIM_Cubic>]}}})))
  ]
  rendererClasses [&NiagaraSpriteRendererProperties]
  renderers [{rendererClass=&NiagaraSpriteRendererProperties}]
 }
 Upward_Beam {
  emitterUpdateScript [
   &EmitterState()
   &SpawnBurst_Instantaneous(Spawn Count=1, Use Spawn Probability=false)
  ]
  particleSpawnScript [
   &InitializeParticle(
    Mesh Scale=<2 0.25 0.25>
    Mesh Scale Mode=::NewEnumerator4
    Position Offset Coordinate Space=::World
    UsePositionOffset=true
    Write Material Random=true
   )
   &AddVelocity(Coordinate Space=::World, Velocity=<0 0 100>) {enabled=false}
  ]
  particleUpdateScript [
   &AccelerationForce(Acceleration=<0 0 1600>) {enabled=false}
   &SolveForcesAndVelocity(Clamp Velocity=false, Limit Acceleration=false) {enabled=false}
   &ParticleState()
   &DynamicMaterialParameters(
    Index 0 Param 1=&FloatFromCurve(FloatCurve=DataInterface() {propertyValues {Curve {keys [
         {arriveTangent=903.074, interpMode=RCIM_Cubic, leaveTangent=903.074, tangentMode=RCTM_User, time=0, value=60}
         {arriveTangentWeight=2.6669, interpMode=RCIM_Cubic, leaveTangentWeight=10.6675, tangentMode=RCTM_User, time=0.1, value=100}
         {arriveTangentWeight=14.0747, interpMode=RCIM_Cubic, leaveTangentWeight=17.5934, tangentMode=RCTM_User, time=0.5, value=100}
         <at 1 5 RCIM_Cubic>
        ]}}})
    Index 0 Param 2=&RandomRangeFloat(Minimum=0, Maximum=1)
    Index 0 Param 3=&RandomRangeFloat(Minimum=0, Maximum=1)
    Module.Param0WriteEnabled=true
    Module.Param1WriteEnabled=true
    Module.Param2WriteEnabled=true
    Module.Param3WriteEnabled=true
   )
   &ScaleColor(
    Scale Alpha=&FloatFromCurve(FloatCurve=DataInterface() {propertyValues {Curve {keys [
         {arriveTangent=21.051, interpMode=RCIM_Cubic, leaveTangent=21.051, tangentMode=RCTM_User, time=0, value=0}
         {arriveTangentWeight=0.0745, interpMode=RCIM_Cubic, leaveTangentWeight=0.2981, tangentMode=RCTM_User, time=0.1, value=1}
         {arriveTangentWeight=0.1993, interpMode=RCIM_Cubic, leaveTangentWeight=0.2491, tangentMode=RCTM_User, time=0.5, value=1}
         <at 1 0 RCIM_Cubic>
        ]}}})
    Scale RGB=&MakeVectorFromLinearColorRGB(Color=&ColorFromCurve(
      ColorCurve=DataInterface() {propertyValues {
        AlphaCurve {keys [<at 0 1 RCIM_Cubic>,<at 1 1 RCIM_Cubic>]}
        BlueCurve {keys [
          <at 0 0.6514 RCIM_Cubic>
          {arriveTangent=0.6972, interpMode=RCIM_Cubic, leaveTangent=0.6972, time=0.25, value=0.9216}
          {arriveTangent=-0.8685, interpMode=RCIM_Cubic, leaveTangent=-0.8685, time=0.5, value=1}
          {arriveTangent=-0.6983, interpMode=RCIM_Cubic, leaveTangent=-0.6983, time=0.75, value=0}
          <at 0.9992 0.6514 RCIM_Cubic>
         ]}
        GreenCurve {keys [
          <at 0 0.0497 RCIM_Cubic>
          {arriveTangent=1.83, interpMode=RCIM_Cubic, leaveTangent=1.83, time=0.25, value=0.0762}
          {arriveTangent=1.2671, interpMode=RCIM_Cubic, leaveTangent=1.2671, time=0.5, value=0.9647}
          {arriveTangent=-1.8329, interpMode=RCIM_Cubic, leaveTangent=-1.8329, time=0.75, value=1}
          <at 0.9992 0.0497 RCIM_Cubic>
         ]}
        RedCurve {keys [
          <at 0 1 RCIM_Cubic>
          {arriveTangent=-2, interpMode=RCIM_Cubic, leaveTangent=-2, time=0.25, value=0.013}
          <at 0.5 0 RCIM_Cubic>
          {arriveTangent=2.0033, interpMode=RCIM_Cubic, leaveTangent=2.0033, time=0.75, value=1}
          <at 0.9992 1 RCIM_Cubic>
         ]}
       }}
      CurveIndex=&Frac_Float(Float=&Multiply_Float(A=$Engine.Time, B=2))
     ))
    ScaleA=true
    ScaleRGB=true
   )
  ]
  rendererClasses [&NiagaraMeshRendererProperties]
  renderers [{rendererClass=&NiagaraMeshRendererProperties, rendererIndex=0}]
 }
 sparks {
  simTarget=GPUComputeSim
  emitterUpdateScript [
   &EmitterState(Life Cycle Mode=::NewEnumerator1, Loop Behavior=::NewEnumerator1, Loop Duration=0.8, UseLoopDelay=false)
   &SpawnRate(SpawnRate=&RandomRangeFloat(Minimum=250, Maximum=500), Use Spawn Probability=false)
  ]
  particleSpawnScript [
   &InitializeParticle(
    Color=&ColorFromCurve(
     ColorCurve=DataInterface() {propertyValues {
       AlphaCurve {keys [<at 0 1 RCIM_Cubic>,<at 1 1 RCIM_Cubic>]}
       BlueCurve {keys [
         <at 0 0.6514 RCIM_Cubic>
         {arriveTangent=0.6972, interpMode=RCIM_Cubic, leaveTangent=0.6972, time=0.25, value=0.9216}
         {arriveTangent=-0.8685, interpMode=RCIM_Cubic, leaveTangent=-0.8685, time=0.5, value=1}
         {arriveTangent=-0.6983, interpMode=RCIM_Cubic, leaveTangent=-0.6983, time=0.75, value=0}
         <at 0.9992 0.6514 RCIM_Cubic>
        ]}
       GreenCurve {keys [
         <at 0 0.0497 RCIM_Cubic>
         {arriveTangent=1.83, interpMode=RCIM_Cubic, leaveTangent=1.83, time=0.25, value=0.0762}
         {arriveTangent=1.2671, interpMode=RCIM_Cubic, leaveTangent=1.2671, time=0.5, value=0.9647}
         {arriveTangent=-1.8329, interpMode=RCIM_Cubic, leaveTangent=-1.8329, time=0.75, value=1}
         <at 0.9992 0.0497 RCIM_Cubic>
        ]}
       RedCurve {keys [
         <at 0 1 RCIM_Cubic>
         {arriveTangent=-2, interpMode=RCIM_Cubic, leaveTangent=-2, time=0.25, value=0.013}
         <at 0.5 0 RCIM_Cubic>
         {arriveTangent=2.0033, interpMode=RCIM_Cubic, leaveTangent=2.0033, time=0.75, value=1}
         <at 0.9992 1 RCIM_Cubic>
        ]}
      }}
     CurveIndex=&RandomRangeFloat(Minimum=0, Maximum=1)
    )
    Lifetime Max=1.5
    Lifetime Min=0.3
    Lifetime Mode=::NewEnumerator1
    Sprite Rotation Mode=::NewEnumerator2
    Sprite Size Mode=::NewEnumerator1
    Uniform Sprite Size Max=15
    Uniform Sprite Size Min=7
    UsePositionOffset=false
    Write Material Random=true
   )
   &SphereLocation(
    MaskSpawnGroup=false
    Override Seed=false
    Sphere Radius=75
    Surface Only=true
    Surface Only Band Thickness=5
    UseNonUniformScale=false
    UseOffset=false
    UseSphereOrientation=false
   )
   &AddVelocityFromPoint(Module.Use Falloff Distance=false)
   &AddVelocity(Coordinate Space=::World, Velocity=<0 0 120>)
  ]
  particleUpdateScript [
   &GravityForce(Gravity=<0 0 -250>)
   &CurlNoiseForce(Noise Frequency=10, Noise Strength=670, Pan Noise=true, Pan Noise Field=<0 0 10>)
   &SolveForcesAndVelocity(Clamp Velocity=false, Limit Acceleration=false)
   &ParticleState()
   &ScaleColor(
    Scale RGB=&VectorFromFloat(Value=&FloatFromCurve(
      FloatCurve=DataInterface() {propertyValues {Curve {keys [
          {arriveTangent=10.9401, interpMode=RCIM_Cubic, leaveTangent=10.9401, tangentMode=RCTM_User, time=0, value=0}
          <at 0.15 1 RCIM_Cubic>
          {arriveTangent=-1.0705, interpMode=RCIM_Cubic, leaveTangent=-1.0705, tangentMode=RCTM_User, time=1, value=0}
         ]}}}
      Scale Curve=&RandomRangeFloat(Minimum=1, Maximum=100, Evaluation Type=::NewEnumerator1)
     ))
    ScaleA=true
    ScaleRGB=true
   )
   &ScaleColor(
    Scale RGB=&MakeVectorFromLinearColorRGB(Color=&ColorFromCurve(
      ColorCurve=DataInterface() {propertyValues {
        AlphaCurve {keys [<at 0 1 RCIM_Cubic>,<at 1 1 RCIM_Cubic>]}
        BlueCurve {keys [
          <at 0 0.6514 RCIM_Cubic>
          {arriveTangent=0.6972, interpMode=RCIM_Cubic, leaveTangent=0.6972, time=0.25, value=0.9216}
          {arriveTangent=-0.8685, interpMode=RCIM_Cubic, leaveTangent=-0.8685, time=0.5, value=1}
          {arriveTangent=-0.6983, interpMode=RCIM_Cubic, leaveTangent=-0.6983, time=0.75, value=0}
          <at 0.9992 0.6514 RCIM_Cubic>
         ]}
        GreenCurve {keys [
          <at 0 0.0497 RCIM_Cubic>
          {arriveTangent=1.83, interpMode=RCIM_Cubic, leaveTangent=1.83, time=0.25, value=0.0762}
          {arriveTangent=1.2671, interpMode=RCIM_Cubic, leaveTangent=1.2671, time=0.5, value=0.9647}
          {arriveTangent=-1.8329, interpMode=RCIM_Cubic, leaveTangent=-1.8329, time=0.75, value=1}
          <at 0.9992 0.0497 RCIM_Cubic>
         ]}
        RedCurve {keys [
          <at 0 1 RCIM_Cubic>
          {arriveTangent=-2, interpMode=RCIM_Cubic, leaveTangent=-2, time=0.25, value=0.013}
          <at 0.5 0 RCIM_Cubic>
          {arriveTangent=2.0033, interpMode=RCIM_Cubic, leaveTangent=2.0033, time=0.75, value=1}
          <at 0.9992 1 RCIM_Cubic>
         ]}
       }}
      CurveIndex=&Frac_Float(Float=&Multiply_Float(A=$Engine.Time, B=2))
     ))
    ScaleA=true
    ScaleRGB=true
   ) {enabled=false, module=ScaleColor001}
   &ScaleSpriteSize(Scale Factor=&Vector2DFromFloat(Value=&FloatFromCurve(FloatCurve=DataInterface() {propertyValues {Curve {keys [
          <at 0 0 RCIM_Cubic>
          {arriveTangent=27.9899, interpMode=RCIM_Cubic, leaveTangent=27.9899, tangentMode=RCTM_User, time=0.1, value=1}
          <at 0.2 1 RCIM_Cubic>
         ]}}})))
  ]
  rendererClasses [&NiagaraSpriteRendererProperties]
  renderers [{rendererClass=&NiagaraSpriteRendererProperties}]
 }
}
summary {
 systemName=NS_Labrador_TeleportOut
 emitters [
  {emitterName=Sweeping_Flares, rendererClasses [&NiagaraSpriteRendererProperties]}
  {emitterName=Corona, rendererClasses [&NiagaraSpriteRendererProperties]}
  {emitterName=Light, rendererClasses [&NiagaraLightRendererProperties]}
  {emitterName=PickUp_Poop, rendererClasses [&NiagaraMeshRendererProperties]}
  {emitterName=Upward_Beam, rendererClasses [&NiagaraMeshRendererProperties]}
  {emitterName=Stretch_Sparks, rendererClasses [&NiagaraSpriteRendererProperties]}
  {emitterName=sparks, simTarget=GPUComputeSim, rendererClasses [&NiagaraSpriteRendererProperties]}
 ]
 userVariables []
}
