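# Physics asset for the Labrador dog skeletal mesh, giving it a simplified collision shape built entirely from single-capsule bodies.
# Capsules cover the head, alternate neck bones, upper and lower body, upper tail chain, two-segment ears, padlock bones and the base of each leg, plus one on the anim bone.
# It defines no constraints, so no joints link the bodies into a ragdoll.
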PhysicsAsset /Labrador/Meshes/Labrador_Mammal_PhysicsAsset
bodies {
 anim [{shapeType=Capsule, length=13.8362, radius=20.1446, rotation=<rot 0 0 172.5076>}]
 backleg_base_l [{shapeType=Capsule, length=17.81, radius=4.2453, center=<11.2483 0.203 0.227>, rotation=<rot 0 91.4892 -88.6588>}]
 backleg_base_r [{shapeType=Capsule, length=17.81, radius=4.2652, center=<-13.2101 -0.4216 0.3565>, rotation=<rot 0 -88.5532 -88.6131>}]
 ear_end_l [{shapeType=Capsule, length=5.9292, radius=2.1441, center=<4.3801 2.1029 0.7329>, rotation=<rot 0 116.6648 -80.6227>}]
 ear_end_r [{shapeType=Capsule, length=5.9292, radius=2.1441, center=<4.3801 2.1029 -0.7329>, rotation=<rot 0 -63.3352 -80.6227>}]
 ear_l [{shapeType=Capsule, length=6.7323, radius=3.0549, rotation=<rot 90 0 0>}]
 ear_r [{shapeType=Capsule, length=6.1466, radius=2.7396, rotation=<rot 90 0 0>}]
 frontleg_base_l [{shapeType=Capsule, length=18.325, radius=4.4469, center=<11.6801 0.2899 0.3066>, rotation=<rot 0 91.0832 -88.3138>}]
 frontleg_base_r [{shapeType=Capsule, length=18.325, radius=4.4355, center=<-12.7781 -0.1667 0.4188>, rotation=<rot 0 -88.9553 -88.2897>}]
 head [{shapeType=Capsule, length=12.415, radius=9.6815, center=<0.1429 5.8921 4.7729>, rotation=<rot 0 -0.5109 -72.2484>}]
 lower_body [{shapeType=Capsule, length=20.8726, radius=15.8841, center=<0.2021 -3.0003 2.1144>, rotation=<rot -12.141 -89.9152 0.5745>}]
 neck_01 [{shapeType=Capsule, length=1.5, radius=12.5, rotation=<rot 90 0 0>}]
 neck_03 [{shapeType=Capsule, length=0.75, radius=10.5, rotation=<rot 90 0 0>}]
 neck_05 [{shapeType=Capsule, length=0.75, radius=9.25, rotation=<rot 90 0 0>}]
 padlock_l [{shapeType=Capsule, length=5.1116, radius=2.6432, center=<1.0284 -0.0462 -4.5289>, rotation=<rot -0.9163 -90 -20.4837>}]
 padlock_r [{shapeType=Capsule, length=5.1149, radius=2.6511, center=<-1.0298 0.0388 4.5314>, rotation=<rot -0.7484 -90 -20.5015>}]
 tail_01 [{shapeType=Capsule, length=5.75, radius=4.5, rotation=<rot 90 0 0>}]
 tail_02 [{shapeType=Capsule, length=5.75, radius=4.5, rotation=<rot 90 0 0>}]
 tail_03 [{shapeType=Capsule, length=5.75, radius=4.5, rotation=<rot 90 0 0>}]
 upper_body [{shapeType=Capsule, length=19.6456, radius=18.1889, center=<0 4.5235 -1.1312>, rotation=<rot -0.8843 -90.3621 -4.5897>}]
}
constraints []
