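# Lit skeletal-mesh material for the Labrador pinata's anticipation effect, adding an animated holographic glow alongside diffuse, normal and mask textures.
# Panning world-aligned smoke is masked by bounding-box height against a timestamp transition (DEBUG switches to manual fade) and suppressed on up- or down-facing normals.
# A Teleporting parameter changes panning and blends toward unmasked, brightened smoke.
# A time-looped curve-atlas color tints the result, which DeriveHDRfromLDR boosts.
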Material /Labrador/Meshes/M_Labrador_Pinata_Anticipation
graph {
 properties {bUsedWithMeshDeformer=true, bUsedWithSkeletalMesh=true, pixelDepthOffsetMode=PDOM_Legacy, usedShadingModels=Default Lit}
 nodes [
  n1 = Multiply(A=Time()) {constB=2}
  n2 = ScalarParameter() {parameterName=Teleporting}
  n3 = MaterialFunctionCall!(
   TextureObject=TextureObject() {texture=&T_Inky_Smoke_Tile_Chromatic}
   TextureSize=VectorParameter() {parameterName=Texture Size, defaultValue=<rgba 100 50 150 0>}.RGB
   WorldPosition=Subtract(A=WorldPosition().XYZ, B=ObjectPositionWS())
   Export Float 4
   In=AppendVector(A=Constant() {r=0.2}, B=LinearInterpolate(Alpha=@n2) {constA=1, constB=2})
   World Space Normal=Transform(None=TextureSample() {samplerType=SAMPLERTYPE_Normal, texture=&/T_Lama_Pinata_SKEL_N}.RGB)
   ProjectionTransitionContrast
  ) {materialFunction=/Game/Athena/SafeZone/WorldAlignedTexture_ScalePan}
  n4 = MaterialFunctionCall() {materialFunction=&BoundingBoxBased_0-1_UVW}
  n5 = ScalarParameter() {defaultValue=1, parameterName=FadeControl}
  n6 = MaterialFunctionCall(
   TimestampA=ScalarParameter() {parameterName=AntiTimeStamp}
   TransitionLength_In=ScalarParameter() {parameterName=AntiTimeLength}
   Transition In Only?=StaticBool() {value=true}
  ) {materialFunction=&MF_Timestamp_InOut_01}
  n7 = Power(Base=OneMinus(None=StaticSwitchParameter(True=@n5, False=@n6.Linear) {parameterName=DEBUG}))
  n8 = Multiply(A=@n3.XY Texture, B=Subtract(A=OneMinus(None=@n4.B), B=Clamp(None=@n7)))
  n9 = VertexNormalWS()
  n10 = Constant3Vector() {constant=<rgba 0 0 1 0>}
  n11 = Constant3Vector() {constant=<rgba 0 0 -1 0>}
  n12 = Multiply(A=OneMinus(None=DotProduct(A=@n9, B=@n10)), B=OneMinus(None=DotProduct(A=@n9, B=@n11)))
  n13 = Multiply(
   A=CurveAtlasRowParameter(CurveTime=Frac(None=@n1)) {atlas=&AthenaCurveAtlas01, curve=/Game/Athena/LinearColorCurves/Curve_C_Holo, defaultValue=250}
   B=LinearInterpolate(A=Multiply(A=@n8, B=@n12), B=Add(A=@n3.XY Texture) {constB=0.25}, Alpha=ScalarParameter() {parameterName=Teleporting})
  )
  MaterialFunctionCall(LDR Input=Clamp(None=@n13), HDR Tint=@n13, HDR Intensity Multiplier=ScalarParameter() {defaultValue=2, parameterName=HDR Intensity}) {materialFunction=&DeriveHDRfromLDR}
  TextureSample() {texture=&/T_Fauna_Labrador_D}
  TextureSample() {samplerType=SAMPLERTYPE_Normal, texture=&T_Lama_Pinata_SKEL_N}
  TextureSample() {samplerType=SAMPLERTYPE_Masks, texture=/Game/FrontEnd/Store/Pinata/Assets/LlamaPinata/Textures/T_Lama_Pinata_InGame_M}
 ]
}
